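# Idle cosmetic Niagara effect for the Flapjack Wrangler body (ZeroPoint variant) that drips particles from its FX_Syrup sockets.
# A single emitter bursts one particle per randomly timed loop, only when grounded in gameplay but always in the front end.
# Each particle picks a random socket, with per-socket mesh scale and slight inverse inherited velocity.
# Gravity ramps over lifetime while curves stretch then collapse the mesh and drive material parameters; rate and distance spawning are disabled.
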NiagaraSystem /BRCosmetics/Characters/Player/Male/Medium/Bodies/M_MED_Flapjack_Wrangler/FX/NS_M_Med_FW_Idle_ZeroPoint
dependencies {
 usedDataInterfaces [&NiagaraDataInterfaceCurve, &NiagaraDataInterfaceArrayFloat, &NiagaraDataInterfaceVectorCurve]
 usedDynamicInputs []
 usedModules [&SystemState]
}
emitters {NE_Empty {
  emitterUpdateScript [
   &EmitterState(
    Life Cycle Mode=::NewEnumerator1
    Loop Duration=&RandomRangeFloat(Minimum=0, Maximum=4)
    Recalculate Duration Each Loop=true
    UseLoopDelay=false
   )
   &SpawnRate(
    SpawnRate=&RandomRangeFloat(Minimum=1, Maximum=4, Evaluation Type=::NewEnumerator1)
    Use Spawn Probability=false
   ) {enabled=false}
   /SetVariables(Emitter.PlayerVelocityVector:Vector3f=$Engine.Owner.Velocity) {isSetParameters=true}
   /SetVariables(Emitter.PlayerSpeed:NiagaraFloat=&VectorLength(Vector=$Emitter.PlayerVelocityVector)) {isSetParameters=true}
   &SpawnPerUnit(
    Module.Use Max Movement Threshold=true
    Module.Use Movement Tolerance=false
    Spawn Per Unit:NiagaraFloat=&Lerp_Float(
     A=40
     B=100
     Alpha=&FloatFromCurve(FloatCurve=DataInterface() {propertyValues {Curve {keys [<at 0 1>,<at 1 0>]}}}, CurveIndex=$Emitter.PlayerSpeed)
    )
    Use Spawn Probability=false
   ) {enabled=false}
   &SpawnBurst_Instantaneous(
    Spawn Count=1
    Spawn Probability=&FN_FrontEndFloatModifier(Gameplay=&MakeCustomFloatFromBool(Bool=$User.IsOnGround, True Float=1, False Float=0), Front End=1)
    Use Spawn Probability=true
   )
  ]
  particleSpawnScript [
   /SetVariables(Particles.WhichSocket:NiagaraInt32=&RandomRangeInt(Minimum=0, Maximum=8)) {isSetParameters=true}
   /Niagara/Modules/Spawn/Initialization/InitializeParticle(
    Mesh Scale=&Multiply_VectorByFloat(
     Vector=&Multiply_VectorByFloat(
      Vector=<1 1 1>
      Float=&Multiply_Float(
       A=0.7
       B=&SelectFloatFromArray(
        Direct Array Index=$Particles.WhichSocket
        Float Selection Array=DataInterface() {propertyValues {FloatData [1,0.7,1.2,1.1,0.8,1.3,1.3,1.25]}}
       )
      )
     )
     Float=&RandomRangeFloat(Minimum=1, Maximum=1.3)
    )
    Module.Write Color=true
    Module.Write Lifetime=true
    Module.Write Mass=true
    Module.Write Material Random=true
    Module.Write Position=false
    Module.Write Scale=true
    Module.Write SpriteRotation=false
    Module.Write SpriteSize=true
    Position=&SimulationPosition()
    Sprite Size=<5 5>
   )
   &SkeletalMeshLocation(
    Bone / Socket Index=$Particles.WhichSocket
    Inherit Velocity=false
    MaskSpawnGroup=false
    Mesh Sampling Type=::NewEnumerator5
    Sampled Position Offset=<0 0 -1>
    Skeletal Mesh=DataInterface() {propertyValues {
      DefaultMesh=null
      ExcludeBoneName=root
      OptionalComponentClass=null
      PreviewMesh=/BRCosmetics/Characters/Player/Male/Medium/Bodies/M_MED_Flapjack_Wrangler/Meshes/Parts/M_MED_Flapjack_Wrangler
      SkinningMode=SkinOnTheFly
      SoftSourceActor=null
      SourceMode=Default
      UvSetIndex=0
      WholeMeshLOD=-1
      bExcludeBone=false
      bReadDeformedGeometry=true
      bRequireCurrentFrameData=true
      ComponentTags []
      FilteredBones []
      FilteredSockets [FX_Syrup_0,FX_Syrup_1,FX_Syrup_2,FX_Syrup_3,FX_Syrup_4,FX_Syrup_5,FX_Syrup_6,FX_Syrup_7]
      MeshUserParameter {parameter {name=None, typeDefHandle {registeredTypeIndex=1}, varData []}}
      SamplingRegions []
     }}
    Socket Sampling Mode=::NewEnumerator5
   )
   &ApplyInitialForces()
   &InheritVelocity(Module.Limit Speed=true, Velocity Scale=&Multiply_VectorByFloat(Vector=<1 1 1>, Float=-0.1)) {module=InheritVelocity001}
   /SetVariables(Particles.SocketFloat:NiagaraFloat=&MakeFloatFromInt(Int=$Particles.WhichSocket)) {isSetParameters=true}
  ]
  particleUpdateScript [
   &ParticleState()
   &GravityForce(Gravity=&Multiply_VectorByFloat(
     Vector=<1 1 -700>
     Float=&FloatFromCurve(FloatCurve=DataInterface() {propertyValues {Curve {keys [{arriveTangent=1.8039, interpMode=RCIM_Cubic, leaveTangent=1.8039, tangentMode=RCTM_User, time=0, value=0}, <at 1 1>]}}})
    ))
   &AccelerationForce(Acceleration=<0 0 -980>) {enabled=false}
   &InheritVelocity(
    Module.Limit Speed=true
    Velocity Limit=150
    Velocity Scale=&Multiply_VectorByFloat(
     Vector=<1 1 1>
     Float=/Niagara/DynamicInputs/Math/LerpMultipleFloats(
      Float Count:ENiagara_MultipleLerpCount=::NewEnumerator3
      1:NiagaraFloat=0.02
      2:NiagaraFloat=0
      3:NiagaraFloat=?
      4:NiagaraFloat=?
      5:NiagaraFloat=?
      Alpha:NiagaraFloat=$Particles.NormalizedAge
     )
    )
   ) {enabled=false}
   &SolveForcesAndVelocity(Clamp Velocity=false, Limit Acceleration=false)
   &ScaleMeshSize(Scale Factor=&Multiply_Vector(
     A=&Multiply_Vector(
      A=<1 1 1>
      B=&VectorFromCurve(VectorCurve=DataInterface() {propertyValues {
         XCurve {keys [
           <at 0 0.5 RCIM_Cubic>
           {arriveTangent=0.5, interpMode=RCIM_Cubic, leaveTangent=-5.3087, tangentMode=RCTM_Break, time=0.1222, value=2.3}
           <at 1 0.5>
          ]}
         YCurve {keys [<at 0 1>,<at 1 1>]}
         ZCurve {keys [<at 0 1>,<at 1 1>]}
        }})
     )
     B=&VectorFromFloat(Value=&FloatFromCurve(FloatCurve=DataInterface() {propertyValues {Curve {keys [<at 0 1>,<at 0.2474 1>,<at 0.5 0>,<at 1 0>]}}}))
    ))
   &ScaleColor(
    Scale Alpha=&FloatFromCurve(FloatCurve=DataInterface() {propertyValues {Curve {keys [<at 0 1 RCIM_Cubic>,<at 1 0>]}}})
    ScaleA=true
    ScaleRGB=true
   ) {enabled=false}
   &DynamicMaterialParameters(
    Index 0 Param 1=0.8
    Index 0 Param 2=&FloatFromCurve(FloatCurve=DataInterface() {propertyValues {Curve {keys [{arriveTangent=-1.8884, interpMode=RCIM_Cubic, leaveTangent=-1.8884, tangentMode=RCTM_User, time=0, value=1}, <at 1 0.26>]}}})
    Index 0 Param 3=1
    Index 0 Param 4=1
    Module.Param0WriteEnabled=true
    Module.Param1WriteEnabled=true
    Module.Param2WriteEnabled=true
    Module.Param3WriteEnabled=true
   )
   &CameraOffset(Camera Offset Amount=5)
  ]
  rendererClasses [&NiagaraSpriteRendererProperties, &NiagaraMeshRendererProperties]
  renderers [
   {rendererClass=&NiagaraSpriteRendererProperties}
   {rendererClass=&NiagaraMeshRendererProperties, rendererIndex=1}
  ]
 }}
summary {
 systemName=NS_M_Med_FW_Idle_ZeroPoint
 emitters [{emitterName=NE_Empty, rendererClasses [&NiagaraSpriteRendererProperties, &NiagaraMeshRendererProperties]}]
 userVariables [
  User.bIsFrontEnd {type=NiagaraBool}
  User.bIsFrontEndPreview {type=NiagaraBool}
  User.IsOnGround {type=NiagaraBool}
  User.JumpSpawn {type=NiagaraFloat, defaultValue=0, description=""}
  User.SkeletalMesh {
   type=NiagaraDataInterfaceSkeletalMesh
   defaultValue {
    struct=NiagaraExt_VariableValue_DataInterface
    value {dataInterface=&self:NiagaraDataInterfaceSkeletalMesh_0}
   }
  }
 ]
}
