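# Physics asset that simulates the front cloth flap and tassels on the Flapjack Wrangler medium male character body.
# Three two-segment capsule chains (left, center, right) hang from tiny MSTR anchor bodies linked to the pelvis, with limited swing and twist.
# The outer lower segments cross-link to the center one to hold the panel together, and seven small tassels hang from the lower segments.
# Larger pelvis and thigh capsules provide body volume for collision.
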PhysicsAsset /BRCosmetics/Characters/Player/Male/Medium/Bodies/M_MED_Flapjack_Wrangler/Meshes/M_MED_Flapjack_Wrangler_Physics_ClothFront
bodies {
 dyn_clothFront1 [{shapeType=Capsule, length=4.2259, radius=1.006, shapeName=dyn_clothFront1_r_capsule, center=<4.8463 0 0>, rotation=<rot 90 0 0>}]
 dyn_clothFront1_MSTR [{shapeType=Capsule, length=0.01, radius=0.01}]
 dyn_clothFront1_l [{shapeType=Capsule, length=4.2259, radius=1.006, shapeName=dyn_clothFront1_r_capsule, center=<4.8463 0 0>, rotation=<rot 90 0 0>}]
 dyn_clothFront1_l_MSTR [{shapeType=Capsule, length=0.01, radius=0.01}]
 dyn_clothFront1_r [{shapeType=Capsule, length=4.2259, radius=1.006, shapeName=dyn_clothFront1_r_capsule, center=<4.8463 0 0>, rotation=<rot 90 0 0>}]
 dyn_clothFront1_r_MSTR [{shapeType=Capsule, length=0.01, radius=0.01}]
 dyn_clothFront2 [{shapeType=Capsule, length=5.5032, radius=1.0358, shapeName=dyn_clothFront2_r_capsule, center=<3.5295 0 0>, rotation=<rot 90 0 0>}]
 dyn_clothFront2_l [{shapeType=Capsule, length=5.5032, radius=1.0358, shapeName=dyn_clothFront2_r_capsule, center=<3.5295 0 0>, rotation=<rot 90 0 0>}]
 dyn_clothFront2_r [{
   shapeType=Capsule
   length=5.5032
   radius=1.0358
   shapeName=dyn_clothFront2_r_capsule
   center=<3.5299 0 0.0168>
   rotation=<rot 87.3084 0.0001 0.0001>
  }]
 dyn_tasselFront1A_l [{shapeType=Capsule, length=1, radius=1.1, shapeName=dyn_tasselBack1C_l_capsule, center=<2.2392 -0.0093 0.0704>, rotation=<rot 90 0 0>}]
 dyn_tasselFront1A_mid_r [{shapeType=Capsule, length=1, radius=1.1, shapeName=dyn_tasselBack1C_l_capsule, center=<2.2392 -0.0093 0.0704>, rotation=<rot 90 0 0>}]
 dyn_tasselFront1A_r [{shapeType=Capsule, length=1, radius=1.1, shapeName=dyn_tasselBack1C_l_capsule, center=<2.2392 -0.0093 0.0704>, rotation=<rot 90 0 0>}]
 dyn_tasselFront1B [{shapeType=Capsule, length=1, radius=1.1, shapeName=dyn_tasselBack1C_l_capsule, center=<2.2392 -0.0093 0.0704>, rotation=<rot 90 0 0>}]
 dyn_tasselFront1C_l [{shapeType=Capsule, length=1, radius=1.1, shapeName=dyn_tasselBack1C_l_capsule, center=<2.2392 -0.0093 0.0704>, rotation=<rot 90 0 0>}]
 dyn_tasselFront1C_mid_l [{shapeType=Capsule, length=1, radius=1.1, shapeName=dyn_tasselBack1C_l_capsule, center=<2.2392 -0.0093 0.0704>, rotation=<rot 90 0 0>}]
 dyn_tasselFront1C_r [{shapeType=Capsule, length=1, radius=1.1, shapeName=dyn_tasselBack1C_l_capsule, center=<2.2392 -0.0093 0.0704>, rotation=<rot 90 0 0>}]
 pelvis [
  {shapeType=Capsule, length=9.5789, radius=11.7449, center=<0.5471 -0.9752 0>}
  {shapeType=Capsule, length=22.25, radius=7.75, center=<-6.9764 -0.1991 0.0001>}
 ]
 thigh_twist_01_l [{shapeType=Capsule, length=27.7107, radius=8.9317, center=<-2.324 -0.8583 -0.0944>, rotation=<rot 90 0 0>}]
 thigh_twist_01_r [{shapeType=Capsule, length=27.5837, radius=8.9573, center=<2.324 -0.4338 -0.7361>, rotation=<rot 90 0 0>}]
}
constraints [
 {
  bone1Name=thigh_twist_01_l
  bone2Name=pelvis
  swing1LimitDegrees=45
  swing1Motion=Free
  swing2LimitDegrees=45
  swing2Motion=Free
  twistLimitDegrees=45
  twistMotion=Free
 }
 {
  bone1Name=thigh_twist_01_r
  bone2Name=pelvis
  swing1LimitDegrees=45
  swing1Motion=Free
  swing2LimitDegrees=45
  swing2Motion=Free
  twistLimitDegrees=45
  twistMotion=Free
 }
 {bone1Name=dyn_clothFront1, bone2Name=dyn_clothFront1_MSTR, swing1LimitDegrees=45, swing2LimitDegrees=45, twistLimitDegrees=10}
 {bone1Name=dyn_clothFront2, bone2Name=dyn_clothFront1, swing1LimitDegrees=25, swing2LimitDegrees=25, twistLimitDegrees=10}
 {bone1Name=dyn_clothFront1_r, bone2Name=dyn_clothFront1_r_MSTR, swing1LimitDegrees=45, swing2LimitDegrees=30, twistLimitDegrees=10}
 {bone1Name=dyn_clothFront2_r, bone2Name=dyn_clothFront1_r, swing1LimitDegrees=25, swing2LimitDegrees=25, twistLimitDegrees=10}
 {bone1Name=dyn_clothFront1_l, bone2Name=dyn_clothFront1_l_MSTR, swing1LimitDegrees=45, swing2LimitDegrees=45, twistLimitDegrees=25}
 {bone1Name=dyn_clothFront2_l, bone2Name=dyn_clothFront1_l, swing1LimitDegrees=15, swing2LimitDegrees=15, twistLimitDegrees=10}
 {bone1Name=dyn_tasselFront1B, bone2Name=dyn_clothFront2, swing1LimitDegrees=45, swing2LimitDegrees=45, twistLimitDegrees=5}
 {
  bone1Name=dyn_clothFront1_l_MSTR
  bone2Name=pelvis
  swing1LimitDegrees=45
  swing1Motion=Free
  swing2LimitDegrees=45
  swing2Motion=Free
  twistLimitDegrees=45
  twistMotion=Free
 }
 {
  bone1Name=dyn_clothFront1_r_MSTR
  bone2Name=pelvis
  swing1LimitDegrees=45
  swing1Motion=Free
  swing2LimitDegrees=45
  swing2Motion=Free
  twistLimitDegrees=45
  twistMotion=Free
 }
 {
  bone1Name=dyn_clothFront1_MSTR
  bone2Name=pelvis
  swing1LimitDegrees=45
  swing1Motion=Free
  swing2LimitDegrees=45
  swing2Motion=Free
  twistLimitDegrees=45
  twistMotion=Free
 }
 {
  bone1Name=dyn_clothFront2_r
  bone2Name=dyn_clothFront2
  swing1LimitDegrees=45
  swing1Motion=Free
  swing2LimitDegrees=45
  swing2Motion=Free
  twistLimitDegrees=30
 }
 {
  bone1Name=dyn_clothFront2_l
  bone2Name=dyn_clothFront2
  swing1LimitDegrees=45
  swing1Motion=Free
  swing2LimitDegrees=45
  swing2Motion=Free
  twistLimitDegrees=30
 }
 {bone1Name=dyn_tasselFront1A_r, bone2Name=dyn_clothFront2_r, swing1LimitDegrees=45, swing2LimitDegrees=45, twistLimitDegrees=5}
 {bone1Name=dyn_tasselFront1A_l, bone2Name=dyn_clothFront2, swing1LimitDegrees=45, swing2LimitDegrees=45, twistLimitDegrees=5}
 {bone1Name=dyn_tasselFront1C_r, bone2Name=dyn_clothFront2, swing1LimitDegrees=45, swing2LimitDegrees=45, twistLimitDegrees=5}
 {bone1Name=dyn_tasselFront1C_l, bone2Name=dyn_clothFront2_l, swing1LimitDegrees=45, swing2LimitDegrees=45, twistLimitDegrees=5}
 {bone1Name=dyn_tasselFront1C_mid_l, bone2Name=dyn_clothFront2_l, swing1LimitDegrees=45, swing2LimitDegrees=45, twistLimitDegrees=5}
 {bone1Name=dyn_tasselFront1A_mid_r, bone2Name=dyn_clothFront2_r, swing1LimitDegrees=45, swing2LimitDegrees=45, twistLimitDegrees=5}
]
