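# Physics asset for the crouching variant of the male medium Flapjack Wrangler's poncho, back cloth and tassels.
# Limited-angle chains form the poncho front and back under spine_02 master bones, and the back cloth under a pelvis master, with five tassels branching from its end.
# Pelvis, spine and thigh capsules supply body shapes, plus a large root capsule centred below the origin, apparently a ground collider.
# Links among body bones and master anchors are free-angle.
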PhysicsAsset /BRCosmetics/Characters/Player/Male/Medium/Bodies/M_MED_Flapjack_Wrangler/Meshes/M_MED_Flapjack_Wrangler_Physics_Poncho_Crouching
bodies {
 dyn_clothBack1 [{shapeType=Capsule, length=19.7096, radius=1.6305, shapeName=dyn_clothBack1_capsule, center=<3.7 0 0>, rotation=<rot 0 0 90>}]
 dyn_clothBack1_MSTR [{shapeType=Capsule, length=0.01, radius=0.01}]
 dyn_clothBack2 [{shapeType=Capsule, length=18.3269, radius=1.4669, shapeName=dyn_clothBack2_capsule, center=<3.1047 0 -0.0923>, rotation=<rot 0 0 -90>}]
 dyn_ponchoBack1 [{shapeType=Capsule, length=6.211, radius=2.3495, shapeName=dyn_ponchoBack1_capsule, center=<3.6852 0 0>, rotation=<rot 0 0 -90>}]
 dyn_ponchoBack2 [{shapeType=Capsule, length=4.0014, radius=2.5653, shapeName=dyn_ponchoBack2_capsule, center=<5.5779 0 0.2039>, rotation=<rot 0 0 -90>}]
 dyn_ponchoBack3 [{shapeType=Capsule, length=5.9171, radius=2.3854, shapeName=dyn_ponchoBack3_capsule, center=<4.7887 0.0882 -0.266>, rotation=<rot 0 0 -90>}]
 dyn_ponchoBack4 [{
   shapeType=Capsule
   length=8.453
   radius=1.0482
   shapeName=dyn_ponchoBack4_capsule
   center=<6.6601 0 -0.2469>
   rotation=<rot 90 179.9975 -0.0025>
  }]
 dyn_ponchoBack_mstr [{shapeType=Sphere, radius=0.505, shapeName=dyn_ponchoBack_mstr_sphere}]
 dyn_ponchoFront1 [{shapeType=Sphere, radius=2, shapeName=dyn_ponchoFront1_sphere}]
 dyn_ponchoFront2 [{shapeType=Capsule, length=2.2862, radius=2.0861, shapeName=dyn_ponchoFront5_capsule, center=<1.94 0 -0.2699>, rotation=<rot 90 0 0>}]
 dyn_ponchoFront3 [{shapeType=Capsule, length=2.2862, radius=2.0861, shapeName=dyn_ponchoFront5_capsule, center=<1.94 0 -0.2699>, rotation=<rot 0 0 90>}]
 dyn_ponchoFront5 [{shapeType=Capsule, length=2.2862, radius=2.0861, shapeName=dyn_ponchoFront5_capsule, center=<1.94 0 -0.2699>, rotation=<rot 90 0 0>}]
 dyn_ponchoFront_mstr [{shapeType=Sphere, radius=0.505, shapeName=dyn_ponchoFront_mstr_sphere}]
 dyn_tasselBack1A_l [{shapeType=Capsule, length=1, radius=1.1, shapeName=dyn_tasselBack1A_r_capsule, center=<2.3626 0.0536 0.0265>, rotation=<rot 90 0 0>}]
 dyn_tasselBack1A_r [{shapeType=Capsule, length=1, radius=1.1, shapeName=dyn_tasselBack1A_r_capsule, center=<2.3626 0.0536 0.0265>, rotation=<rot 90 0 0>}]
 dyn_tasselBack1B [{shapeType=Capsule, length=2.25, radius=1, center=<1.6466 0 -0.0002>, rotation=<rot -90 -7.5524 7.5525>}]
 dyn_tasselBack1C_l [{shapeType=Capsule, length=1, radius=1.1, shapeName=dyn_tasselBack1C_l_capsule, center=<2.2392 -0.0093 0.0704>, rotation=<rot 90 0 0>}]
 dyn_tasselBack1C_r [{shapeType=Capsule, length=1, radius=1.1, shapeName=dyn_tasselBack1A_r_capsule, center=<2.3626 0.0536 0.0265>, rotation=<rot 90 0 0>}]
 pelvis [
  {shapeType=Capsule, length=10.8356, radius=10.4882, center=<0.5471 -0.9752 0>}
  {shapeType=Capsule, length=22.25, radius=7.75, center=<-6.9764 -0.1991 0.0001>}
 ]
 root [{shapeType=Capsule, length=34.7081, radius=76.1545, shapeName=root_capsule, center=<0 0 -64.0652>, rotation=<rot 90 0 0>}]
 spine_02 [{shapeType=Capsule, length=15.8945, radius=10.3201, center=<0 -0.3769 0>}]
 thigh_twist_01_l [{shapeType=Capsule, length=27.7107, radius=8.9317, center=<-2.324 -0.8583 -0.0944>, rotation=<rot 90 0 0>}]
 thigh_twist_01_r [{shapeType=Capsule, length=25.9253, radius=10.6157, center=<2.324 -0.0735 -0.7361>, rotation=<rot 90 0 0>}]
}
constraints [
 {bone1Name=dyn_ponchoBack1, bone2Name=dyn_ponchoBack_mstr, swing1LimitDegrees=3, swing2LimitDegrees=25, twistLimitDegrees=5}
 {bone1Name=dyn_ponchoBack2, bone2Name=dyn_ponchoBack1, swing1LimitDegrees=10, swing2LimitDegrees=30, twistLimitDegrees=5}
 {bone1Name=dyn_ponchoBack4, bone2Name=dyn_ponchoBack3, swing1LimitDegrees=15, swing2LimitDegrees=15, twistLimitDegrees=5}
 {
  bone1Name=dyn_ponchoFront_mstr
  bone2Name=spine_02
  swing1LimitDegrees=45
  swing1Motion=Free
  swing2LimitDegrees=45
  swing2Motion=Free
  twistLimitDegrees=45
  twistMotion=Free
 }
 {
  bone1Name=dyn_ponchoBack_mstr
  bone2Name=spine_02
  swing1LimitDegrees=45
  swing1Motion=Free
  swing2LimitDegrees=45
  swing2Motion=Free
  twistLimitDegrees=45
  twistMotion=Free
 }
 {
  bone1Name=spine_02
  bone2Name=pelvis
  swing1LimitDegrees=45
  swing1Motion=Free
  swing2LimitDegrees=45
  swing2Motion=Free
  twistLimitDegrees=45
  twistMotion=Free
 }
 {
  bone1Name=thigh_twist_01_l
  bone2Name=pelvis
  swing1LimitDegrees=45
  swing1Motion=Free
  swing2LimitDegrees=45
  swing2Motion=Free
  twistLimitDegrees=45
  twistMotion=Free
 }
 {
  bone1Name=thigh_twist_01_r
  bone2Name=pelvis
  swing1LimitDegrees=45
  swing1Motion=Free
  swing2LimitDegrees=45
  swing2Motion=Free
  twistLimitDegrees=45
  twistMotion=Free
 }
 {bone1Name=dyn_clothBack1, bone2Name=dyn_clothBack1_MSTR, swing1LimitDegrees=40, swing2LimitDegrees=35, twistLimitDegrees=10}
 {bone1Name=dyn_clothBack2, bone2Name=dyn_clothBack1, swing1LimitDegrees=5, swing2LimitDegrees=45, twistLimitDegrees=10}
 {bone1Name=dyn_tasselBack1B, bone2Name=dyn_clothBack2, swing1LimitDegrees=45, swing2LimitDegrees=45, twistLimitDegrees=5}
 {
  bone1Name=dyn_clothBack1_MSTR
  bone2Name=pelvis
  swing1LimitDegrees=45
  swing1Motion=Free
  swing2LimitDegrees=45
  swing2Motion=Free
  twistLimitDegrees=45
  twistMotion=Free
 }
 {bone1Name=dyn_tasselBack1C_l, bone2Name=dyn_clothBack2, swing1LimitDegrees=45, swing2LimitDegrees=45, twistLimitDegrees=5}
 {bone1Name=dyn_tasselBack1A_r, bone2Name=dyn_clothBack2, swing1LimitDegrees=45, swing2LimitDegrees=45, twistLimitDegrees=5}
 {bone1Name=dyn_tasselBack1A_l, bone2Name=dyn_clothBack2, swing1LimitDegrees=45, swing2LimitDegrees=45, twistLimitDegrees=5}
 {bone1Name=dyn_tasselBack1C_r, bone2Name=dyn_clothBack2, swing1LimitDegrees=45, swing2LimitDegrees=45, twistLimitDegrees=5}
 {bone1Name=dyn_ponchoBack3, bone2Name=dyn_ponchoBack2, swing1LimitDegrees=10, swing2LimitDegrees=30, twistLimitDegrees=5}
 {
  bone1Name=dyn_ponchoFront1
  bone2Name=dyn_ponchoFront_mstr
  swing1LimitDegrees=45
  swing1Motion=Free
  swing2LimitDegrees=45
  swing2Motion=Free
  twistLimitDegrees=45
  twistMotion=Free
 }
 {bone1Name=dyn_ponchoFront2, bone2Name=dyn_ponchoFront1, swing1LimitDegrees=10, swing2LimitDegrees=60, twistLimitDegrees=15}
 {bone1Name=dyn_ponchoFront3, bone2Name=dyn_ponchoFront2, swing1LimitDegrees=10, swing2LimitDegrees=60, twistLimitDegrees=15}
 {bone1Name=dyn_ponchoFront5, bone2Name=dyn_ponchoFront3, swing1LimitDegrees=15, swing2LimitDegrees=45, twistLimitDegrees=15}
 {
  bone1Name=pelvis
  bone2Name=root
  swing1LimitDegrees=45
  swing1Motion=Free
  swing2LimitDegrees=45
  swing2Motion=Free
  twistLimitDegrees=45
  twistMotion=Free
 }
]
