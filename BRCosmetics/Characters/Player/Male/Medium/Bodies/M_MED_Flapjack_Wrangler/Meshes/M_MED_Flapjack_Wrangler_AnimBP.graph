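# Post-body-pose AnimBlueprint for the Flapjack Wrangler cosmetic, copying the parent mesh pose and adding cloth, poncho, tassel and holster secondary motion.
# Crouch, front-end and default branches pair RigidBody simulations with bone correctives.
# A shared tail adds tassel constraints, AnimDynamics holsters and a sprint back-cloth offset.
# Each update derives corrective alphas from thigh rotation and pawn speed, and feeds skydiving force into gameplay simulations as overridden gravity.
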AnimBlueprint /BRCosmetics/Characters/Player/Male/Medium/Bodies/M_MED_Flapjack_Wrangler/Meshes/M_MED_Flapjack_Wrangler_AnimBP
target_skeleton=&/M_MED_Flapjack_Wrangler_Skeleton
graphs [
 AnimGraph {nodes [
   n1 = Look At - Bone: dyn_clothFront1_r_MSTR(ComponentPose=Local To Component(LocalPose=Copy Pose From Mesh() {bUseAttachedParent=true}), Alpha=1) {
    alphaBoolBlend {blendOption=Linear}
    boneToModify {boneName=dyn_clothFront1_r_MSTR}
    lookAtLocation=<0 -10>
    lookAtTarget {boneReference {boneName=calf_r}}
    lookAt_Axis {axis=<1 0>}
   }
   n2 = Look At - Bone: dyn_clothFront1_l_MSTR(ComponentPose=@n1, Alpha=1) {
    alphaBoolBlend {blendOption=Linear}
    boneToModify {boneName=dyn_clothFront1_l_MSTR}
    lookAtLocation=<0 10>
    lookAtTarget {boneReference {boneName=calf_l}}
    lookAt_Axis {axis=<1 0>}
   }
   n3 = RigidBody!(
    ComponentPose=@n2
    Alpha=Get Skydiving_alpha() -> (Skydiving_alpha:double)
    OverrideWorldGravity:Vector=Get windVector() -> (windVector:Vector)
    ExternalForce
   ) -> (Pose) {
    bClampLinearTranslationLimitToRefPose=true
    bOverrideWorldGravity=true
    overridePhysicsAsset=&/M_MED_Flapjack_Wrangler_Physics_ClothFront
    simulationSpace=BaseBoneSpace
    alphaBoolBlend {blendOption=Linear}
    alphaScaleBiasClamp {outRange {min=0.5}}
    baseBoneRef {boneName=pelvis}
    componentAppliedLinearAccClamp=<5000 5000 5000>
    componentLinearAccScale=<0.35 0.35 0.35>
    componentLinearVelScale=<0.3 0.3 0.3>
    overrideWorldGravity {z=-980}
   }
   n4 = Get Thigh_l_alpha() -> (Thigh_l_alpha:double)
   n5 = "Transform (Modify) Bone - Bone: dyn_clothFront1_l_MSTR"(ComponentPose=@n3, Alpha=@n4, Translation="-2,0.000,2", Rotation="0,0,0", Scale="1,1,1") {
    rotationMode=BMM_Additive
    rotationSpace=BCS_BoneSpace
    translationMode=BMM_Additive
    translationSpace=BCS_BoneSpace
    alphaBoolBlend {blendOption=Linear}
    boneToModify {boneName=dyn_clothFront1_l_MSTR}
    translation {x=-2, z=2}
   }
   n6 = Get Thigh_r_alpha() -> (Thigh_r_alpha:double)
   n7 = "Transform (Modify) Bone - Bone: dyn_clothFront1_r_MSTR"(ComponentPose=@n5, Alpha=@n6, Translation="-2,0.000,5", Rotation="0,0,0", Scale="1,1,1") {
    rotationMode=BMM_Additive
    rotationSpace=BCS_BoneSpace
    translationMode=BMM_Additive
    translationSpace=BCS_BoneSpace
    alphaBoolBlend {blendOption=Linear}
    boneToModify {boneName=dyn_clothFront1_r_MSTR}
    translation {x=-2, z=5}
   }
   n8 = "Transform (Modify) Bone - Bone: dyn_clothFront1_r_MSTR"(
    ComponentPose=@n7
    Alpha=Get Movement_alpha_interpolation() -> (Movement_alpha_interpolation:double)
    Translation="-3,0.000,0"
    Rotation="0,0,0"
    Scale="1,1,1"
   ) {
    rotationSpace=BCS_BoneSpace
    translationMode=BMM_Additive
    translationSpace=BCS_BoneSpace
    alphaBoolBlend {blendOption=Linear}
    boneToModify {boneName=dyn_clothFront1_r_MSTR}
    translation {x=-3}
   }
   n9 = "Transform (Modify) Bone - Bone: dyn_clothFront2_l"(ComponentPose=@n8, Alpha=@n4, Translation="0,0,0", Rotation="-20,0.000,0", Scale="1,1,1") {
    rotationMode=BMM_Additive
    rotationSpace=BCS_BoneSpace
    alphaBoolBlend {blendOption=Linear}
    boneToModify {boneName=dyn_clothFront2_l}
    rotation {pitch=-20}
   }
   n10 = "Transform (Modify) Bone - Bone: dyn_clothFront2_r"(ComponentPose=@n9, Alpha=@n6, Translation="0,0,0", Rotation="-40,0.000,0", Scale="1,1,1") {
    rotationMode=BMM_Additive
    rotationSpace=BCS_BoneSpace
    alphaBoolBlend {blendOption=Linear}
    boneToModify {boneName=dyn_clothFront2_r}
    rotation {pitch=-40}
   }
   n11 = "Transform (Modify) Bone - Bone: dyn_clothFront1_l_MSTR"(ComponentPose=@n10, Alpha=1, Translation="0,0,0", Rotation="0.000,-3,-0.000", Scale="1,1,1") {
    rotationMode=BMM_Additive
    rotationSpace=BCS_BoneSpace
    translationSpace=BCS_BoneSpace
    alphaBoolBlend {blendOption=Linear}
    boneToModify {boneName=dyn_clothFront1_l_MSTR}
    rotation {yaw=-3}
   }
   n12 = "Transform (Modify) Bone - Bone: dyn_clothFront1_r_MSTR"(ComponentPose=@n11, Alpha=1, Translation="0,0,0", Rotation="0.000,3,-0.000", Scale="1,1,1") {
    rotationMode=BMM_Additive
    rotationSpace=BCS_BoneSpace
    translationSpace=BCS_BoneSpace
    alphaBoolBlend {blendOption=Linear}
    boneToModify {boneName=dyn_clothFront1_r_MSTR}
    rotation {yaw=3}
   }
   n13 = RigidBody(ComponentPose=@n12, OverrideWorldGravity=Get windVector()) {
    bClampLinearTranslationLimitToRefPose=true
    bOverrideWorldGravity=true
    overridePhysicsAsset=&/M_MED_Flapjack_Wrangler_Physics_Poncho
    simulationSpace=BaseBoneSpace
    alphaBoolBlend {blendOption=Linear}
    baseBoneRef {boneName=pelvis}
    componentAppliedLinearAccClamp=<5000 5000 5000>
    componentLinearAccScale=<0.35 0.35 0.35>
    componentLinearVelScale=<0.3 0.3 0.3>
    overrideWorldGravity {z=-980}
   }
   n14 = RigidBody!(ComponentPose=Local To Component(LocalPose=Copy Pose From Mesh() {bUseAttachedParent=true}), Alpha, ExternalForce) -> (Pose) {
    bOverrideWorldGravity=true
    overridePhysicsAsset=&M_MED_Flapjack_Wrangler_Physics_ClothFront
    simulationSpace=BaseBoneSpace
    alphaBoolBlend {blendOption=Linear}
    baseBoneRef {boneName=pelvis}
    overrideWorldGravity {z=-300}
    simSpaceSettings {maxAngularAcceleration=300, maxAngularVelocity=5, maxLinearAcceleration=400, maxLinearVelocity=250, velocityScaleZ=0.7, worldAlpha=1}
   }
   n15 = Get Thigh_l_alpha()
   n16 = "Transform (Modify) Bone - Bone: dyn_clothFront1_l_MSTR"(ComponentPose=@n14, Alpha=@n15, Translation="-2,0.000,2", Rotation="0,0,0", Scale="1,1,1") {
    rotationMode=BMM_Additive
    rotationSpace=BCS_BoneSpace
    translationMode=BMM_Additive
    translationSpace=BCS_BoneSpace
    alphaBoolBlend {blendOption=Linear}
    boneToModify {boneName=dyn_clothFront1_l_MSTR}
    translation {x=-2, z=2}
   }
   n17 = Get Thigh_r_alpha()
   n18 = "Transform (Modify) Bone - Bone: dyn_clothFront1_r_MSTR"(ComponentPose=@n16, Alpha=@n17, Translation="-2,0.000,5", Rotation="0,0,0", Scale="1,1,1") {
    rotationMode=BMM_Additive
    rotationSpace=BCS_BoneSpace
    translationMode=BMM_Additive
    translationSpace=BCS_BoneSpace
    alphaBoolBlend {blendOption=Linear}
    boneToModify {boneName=dyn_clothFront1_r_MSTR}
    translation {x=-2, z=5}
   }
   n19 = "Transform (Modify) Bone - Bone: dyn_clothFront1_r_MSTR"(ComponentPose=@n18, Alpha=Get Movement_alpha_interpolation(), Translation="-3,0.000,0", Rotation="0,0,0", Scale="1,1,1") {
    rotationSpace=BCS_BoneSpace
    translationMode=BMM_Additive
    translationSpace=BCS_BoneSpace
    alphaBoolBlend {blendOption=Linear}
    boneToModify {boneName=dyn_clothFront1_r_MSTR}
    translation {x=-3}
   }
   n20 = "Transform (Modify) Bone - Bone: dyn_clothFront2_l"(ComponentPose=@n19, Alpha=@n15, Translation="0,0,0", Rotation="-20,0.000,0", Scale="1,1,1") {
    rotationMode=BMM_Additive
    rotationSpace=BCS_BoneSpace
    alphaBoolBlend {blendOption=Linear}
    boneToModify {boneName=dyn_clothFront2_l}
    rotation {pitch=-20}
   }
   n21 = "Transform (Modify) Bone - Bone: dyn_clothFront2_r"(ComponentPose=@n20, Alpha=@n17, Translation="0,0,0", Rotation="-40,0.000,0", Scale="1,1,1") {
    rotationMode=BMM_Additive
    rotationSpace=BCS_BoneSpace
    alphaBoolBlend {blendOption=Linear}
    boneToModify {boneName=dyn_clothFront2_r}
    rotation {pitch=-40}
   }
   n22 = "Transform (Modify) Bone - Bone: dyn_clothFront1_l_MSTR"(ComponentPose=@n21, Alpha=1, Translation="0,0,0", Rotation="0.000,-3,-0.000", Scale="1,1,1") {
    rotationMode=BMM_Additive
    rotationSpace=BCS_BoneSpace
    translationSpace=BCS_BoneSpace
    alphaBoolBlend {blendOption=Linear}
    boneToModify {boneName=dyn_clothFront1_l_MSTR}
    rotation {yaw=-3}
   }
   n23 = "Transform (Modify) Bone - Bone: dyn_clothFront1_r_MSTR"(ComponentPose=@n22, Alpha=1, Translation="0,0,0", Rotation="0.000,3,-0.000", Scale="1,1,1") {
    rotationMode=BMM_Additive
    rotationSpace=BCS_BoneSpace
    translationSpace=BCS_BoneSpace
    alphaBoolBlend {blendOption=Linear}
    boneToModify {boneName=dyn_clothFront1_r_MSTR}
    rotation {yaw=3}
   }
   n24 = RigidBody(ComponentPose=@n23) {
    bOverrideWorldGravity=true
    overridePhysicsAsset=&M_MED_Flapjack_Wrangler_Physics_Poncho
    simulationSpace=BaseBoneSpace
    alphaBoolBlend {blendOption=Linear}
    baseBoneRef {boneName=pelvis}
    overrideWorldGravity {z=-500}
    simSpaceSettings {maxAngularAcceleration=300, maxAngularVelocity=5, maxLinearAcceleration=400, maxLinearVelocity=250, velocityScaleZ=0.7, worldAlpha=1}
   }
   n25 = Blend Poses by bool(BlendPose_0=Component To Local(ComponentPose=@n24), BlendPose_1=Component To Local(ComponentPose=@n13), bActiveValue=Get bIsInFrontEnd()) {blendType=Linear}
   n26 = Get windVector()
   n27 = "Transform (Modify) Bone - Bone: dyn_clothFront1_MSTR"(
    ComponentPose=Local To Component(LocalPose=Copy Pose From Mesh() {bUseAttachedParent=true})
    Alpha=Get Thigh_L_crouch_alpha_RESERV() -> (Thigh_L_crouch_alpha_RESERV:double)
    Translation="-12,0,0"
    Rotation="0,0,0"
    Scale="1,1,1"
   ) {
    translationMode=BMM_Additive
    translationSpace=BCS_BoneSpace
    alphaBoolBlend {blendOption=Linear}
    alphaScaleBiasClamp {bMapRange=true, outRange {min=0.3}}
    boneToModify {boneName=dyn_clothFront1_MSTR}
    translation {x=-12}
   }
   n28 = "Transform (Modify) Bone - Bone: dyn_clothFront1_l_MSTR"(ComponentPose=@n27, Alpha=1, Translation="-4,6,5", Rotation="0,0,0", Scale="1,1,1") {
    translationMode=BMM_Additive
    translationSpace=BCS_BoneSpace
    alphaBoolBlend {blendOption=Linear}
    boneToModify {boneName=dyn_clothFront1_l_MSTR}
    translation=<-4 6 5>
   }
   n29 = "Transform (Modify) Bone - Bone: dyn_clothFront1_l_MSTR"(
    ComponentPose=@n28
    Alpha=Get Thigh_L_crouch_alpha_RESERV()
    Translation=Get Thigh_L_transform_crouching() -> (Thigh_L_transform_crouching:Vector)
    Rotation="0,0,0"
    Scale="1,1,1"
   ) {
    translationMode=BMM_Additive
    translationSpace=BCS_BoneSpace
    alphaBoolBlend {blendOption=Linear}
    boneToModify {boneName=dyn_clothFront1_l_MSTR}
   }
   n30 = "Transform (Modify) Bone - Bone: dyn_clothFront1_r_MSTR"(
    ComponentPose=@n29
    Alpha=Get Thigh_r_crouch_alpha() -> (Thigh_r_crouch_alpha:double)
    Translation="-17,-6,4"
    Rotation="0,0,0"
    Scale="1,1,1"
   ) {
    rotationMode=BMM_Additive
    rotationSpace=BCS_BoneSpace
    translationMode=BMM_Additive
    translationSpace=BCS_BoneSpace
    alphaBoolBlend {blendOption=Linear}
    boneToModify {boneName=dyn_clothFront1_r_MSTR}
    translation=<-17 -6 4>
   }
   n31 = RigidBody!(ComponentPose=@n30, bAlphaBoolEnabled:bool, OverrideWorldGravity:Vector=@n26, ExternalForce) -> (Pose) {
    alphaInputType=Bool
    bClampLinearTranslationLimitToRefPose=true
    bOverrideWorldGravity=true
    overridePhysicsAsset=&/M_MED_Flapjack_Wrangler_Physics_ClothFrontCrouch
    simulationSpace=BaseBoneSpace
    alphaBoolBlend {blendOption=Linear}
    baseBoneRef {boneName=pelvis}
    componentAppliedLinearAccClamp=<5000 5000 5000>
    componentLinearAccScale=<0.35 0.35 0.35>
    componentLinearVelScale=<0.3 0.3 0.3>
    overrideWorldGravity {z=-980}
   }
   n32 = Get Movement_alpha_interpolation()
   n33 = "Transform (Modify) Bone - Bone: dyn_ponchoFront_mstr"(ComponentPose=@n31, Alpha=@n32, Translation="0,0,0", Rotation="10,0,0", Scale="1,1,1") {
    rotationMode=BMM_Replace
    rotationSpace=BCS_BoneSpace
    alphaBoolBlend {blendOption=Linear}
    alphaScaleBiasClamp {bMapRange=true, outRange {min=0.3}}
    boneToModify {boneName=dyn_ponchoFront_mstr}
    rotation {pitch=10}
   }
   n34 = "Constraint - dyn_ponchoFront1 (RR)"!(ComponentPose=@n33, Alpha=@n32, ConstraintWeights_0=0.5, ConstraintWeights_1:float=0.5) -> (Pose) {
    alphaBoolBlend {blendOption=Linear}
    alphaScaleBiasClamp {bMapRange=true, outRange {max=0.7}}
    boneToModify {boneName=dyn_ponchoFront1}
    constraintSetup [{transformType=Rotation, targetBone {boneName=thigh_l}}, {transformType=Rotation, targetBone {boneName=thigh_r}}]
    constraintWeights [1,1]
   }
   n35 = RigidBody!(ComponentPose=@n34, Alpha:float, OverrideWorldGravity=@n26, ExternalForce) -> (Pose) {
    bClampLinearTranslationLimitToRefPose=true
    bOverrideWorldGravity=true
    overridePhysicsAsset=&/M_MED_Flapjack_Wrangler_Physics_Poncho_Crouching
    simulationSpace=BaseBoneSpace
    alphaBoolBlend {blendOption=Linear}
    baseBoneRef {boneName=pelvis}
    componentAppliedLinearAccClamp=<5000 5000 5000>
    componentLinearAccScale=<0.35 0.35 0.35>
    componentLinearVelScale=<0.3 0.3 0.3>
    overrideWorldGravity {z=-980}
   }
   n36 = Get Movement_alpha_interpolation()
   n37 = "Transform (Modify) Bone - Bone: dyn_ponchoFront3"(ComponentPose=@n35, Alpha=@n36, Translation="0,0,0", Rotation="20,0.000,0.000", Scale="1,1,1") {
    rotationMode=BMM_Additive
    rotationSpace=BCS_BoneSpace
    alphaBoolBlend {blendOption=Linear}
    alphaScaleBiasClamp {bMapRange=true, outRange {max=0, min=1}}
    boneToModify {boneName=dyn_ponchoFront3}
    rotation {pitch=20}
   }
   n38 = "Transform (Modify) Bone - Bone: dyn_ponchoFront4"(ComponentPose=@n37, Alpha=@n36, Translation="0,0,0", Rotation="10,0.000,0.000", Scale="1,1,1") {
    rotationMode=BMM_Additive
    rotationSpace=BCS_BoneSpace
    alphaBoolBlend {blendOption=Linear}
    alphaScaleBiasClamp {bMapRange=true, outRange {max=0, min=1}}
    boneToModify {boneName=dyn_ponchoFront4}
    rotation {pitch=10}
   }
   n39 = "Transform (Modify) Bone - Bone: dyn_ponchoFront5"(ComponentPose=@n38, Alpha=@n36, Translation="0,0,0", Rotation="-15,0.000,0.000", Scale="1,1,1") {
    rotationMode=BMM_Additive
    rotationSpace=BCS_BoneSpace
    alphaBoolBlend {blendOption=Linear}
    alphaScaleBiasClamp {bMapRange=true, outRange {max=0, min=1}}
    boneToModify {boneName=dyn_ponchoFront5}
    rotation {pitch=-15}
   }
   n40 = Component To Local(ComponentPose=@n39)
   n41 = Get bIsCrouching()
   n42 = "Transform (Modify) Bone - Bone: dyn_clothFront1_l"(
    ComponentPose=Local To Component(LocalPose=Blend Poses by bool(BlendPose_0=@n40, BlendPose_1=@n25, bActiveValue=@n41) {blendType=Linear})
    Alpha=1
    Translation="-0.000,-0.004,2.311"
    Rotation="0,0,0"
    Scale="1,1,1"
   ) {
    rotationMode=BMM_Additive
    rotationSpace=BCS_BoneSpace
    scaleSpace=BCS_BoneSpace
    translationMode=BMM_Additive
    translationSpace=BCS_BoneSpace
    alphaBoolBlend {blendOption=Linear}
    boneToModify {boneName=dyn_clothFront1_l}
    translation {y=-0.004, z=2.311}
   }
   n43 = "Transform (Modify) Bone - Bone: dyn_clothFront1_r"(ComponentPose=@n42, Alpha=1, Translation="0.000,0.015,1", Rotation="0,0,0", Scale="1,1,1") {
    rotationMode=BMM_Additive
    rotationSpace=BCS_BoneSpace
    scaleSpace=BCS_BoneSpace
    translationMode=BMM_Additive
    translationSpace=BCS_BoneSpace
    alphaBoolBlend {blendOption=Linear}
    boneToModify {boneName=dyn_clothFront1_r}
    translation {y=0.015, z=1}
   }
   n44 = "Constraint - dyn_tasselFront1C_r (PP)"!(ComponentPose=@n43, Alpha=0.5, ConstraintWeights_0=0.5, ConstraintWeights_1:float=0.5) -> (Pose) {
    alphaBoolBlend {blendOption=Linear}
    boneToModify {boneName=dyn_tasselFront1C_r}
    constraintSetup [{perAxis {bZ=false}, targetBone {boneName=dyn_clothFront2_r}}, {perAxis {bZ=false}, targetBone {boneName=dyn_clothFront2}}]
    constraintWeights [1,1]
   }
   n45 = "Constraint - dyn_tasselFront1A_l (PP)"!(ComponentPose=@n44, Alpha=0.5, ConstraintWeights_0=0.5, ConstraintWeights_1:float=0.5) -> (Pose) {
    alphaBoolBlend {blendOption=Linear}
    boneToModify {boneName=dyn_tasselFront1A_l}
    constraintSetup [{perAxis {bZ=false}, targetBone {boneName=dyn_clothFront2_l}}, {perAxis {bZ=false}, targetBone {boneName=dyn_clothFront2}}]
    constraintWeights [1,1]
   }
   n46 = "Constraint - dyn_tasselFront1C_mid_l (PP)"!(ComponentPose=@n45, Alpha=0.5, ConstraintWeights_0=0.5, ConstraintWeights_1:float=0.5) -> (Pose) {
    alphaBoolBlend {blendOption=Linear}
    boneToModify {boneName=dyn_tasselFront1C_mid_l}
    constraintSetup [{perAxis {bZ=false}, targetBone {boneName=dyn_clothFront2_l}}, {perAxis {bZ=false}, targetBone {boneName=dyn_clothFront2}}]
    constraintWeights [1,1]
   }
   n47 = "Constraint - dyn_tasselFront1A_mid_r (PP)"!(ComponentPose=@n46, Alpha=0.5, ConstraintWeights_0=0.5, ConstraintWeights_1:float=0.5) -> (Pose) {
    alphaBoolBlend {blendOption=Linear}
    boneToModify {boneName=dyn_tasselFront1A_mid_r}
    constraintSetup [{perAxis {bZ=false}, targetBone {boneName=dyn_clothFront2_r}}, {perAxis {bZ=false}, targetBone {boneName=dyn_clothFront2}}]
    constraintWeights [1,1]
   }
   n48 = Anim Dynamics - Bone: dyn_holster1_l(ComponentPose=@n47, Alpha=1, ExternalForce="0,0,0") {
    angularDampingOverride=0.9
    bOverrideAngularDamping=true
    bUseGravityOverride=true
    alphaBoolBlend {blendOption=Linear}
    boundBone {boneName=dyn_holster1_l}
    gravityOverride {z=-980}
    physicsBodyDefinitions [{
      boundBone {boneName=dyn_holster1_l}
      boxExtents=<15 10 5>
      constraintSetup {angularLimitsMax=<5 1 0>, angularLimitsMin=<-25 -1 0>, angularTarget=<0 0 0>}
      localJointOffset=<5 0 0>
     }]
   }
   n49 = Anim Dynamics - Bone: dyn_holster1_r(ComponentPose=@n48, Alpha=1, ExternalForce="0,0,0") {
    angularDampingOverride=0.9
    bOverrideAngularDamping=true
    bUseGravityOverride=true
    alphaBoolBlend {blendOption=Linear}
    boundBone {boneName=dyn_holster1_r}
    gravityOverride {z=-980}
    physicsBodyDefinitions [{
      boundBone {boneName=dyn_holster1_r}
      boxExtents=<15 10 5>
      constraintSetup {angularLimitsMax=<5 1 0>, angularLimitsMin=<-25 -1 0>, angularTarget=<0 0 0>}
      localJointOffset=<5 0 0>
     }]
   }
   n50 = "Transform (Modify) Bone - Bone: dyn_clothBack1_MSTR"!(ComponentPose=@n49, bAlphaBoolEnabled:bool=Get bIsSprinting(), Translation="0,0,1.5", Rotation="-0.000,-0.000,-4", Scale="1,1,1") -> (Pose) {
    alphaInputType=Bool
    rotationMode=BMM_Additive
    rotationSpace=BCS_BoneSpace
    scaleSpace=BCS_BoneSpace
    translationMode=BMM_Additive
    translationSpace=BCS_BoneSpace
    alphaBoolBlend {blendInTime=0.3, blendOption=Linear, blendOutTime=0.3}
    boneToModify {boneName=dyn_clothBack1_MSTR}
    rotation {roll=-4}
    translation {z=1.5}
   }
   Output Pose(Result=Component To Local(ComponentPose=@n50)) {layerGroup=None}
  ]}
 EventGraph {nodes [
   n1 = Event BlueprintUpdateAnimation()
   > Sequence()
   n2 = Break Fort Anim Input Skydiving External Force(FortAnimInput_SkydivingExternalForce=Get SkydivingExternalForceInput()) -> (
    bUseSkydivingVectorForce=~
    SkydivingVectorForce=~
    HeadToPelvisDirection=~
    FloatingMultiplier=~
    FloatingAdditive=~
    DivingMultiplier=~
    DivingAdditive=~
    ParachutingMultiplier=~
    ParachutingAdditive=~
   )
   n3 = Get bIsSkydiveFloating()
   n4 = Get bIsSkydiveDiveMode()
   n5 = Get bIsSkyDiving()
   > Set windVector(windVector:Vector=SelectVector(A=vector * float(A=@n2.SkydivingVectorForce, B=0.15), B="0, 0,-980", bPickA=OR Boolean!(A=@n5, B=@n3, C:bool=@n4) -> (ReturnValue))) -> (Output_Get:Vector)
   n6 = "float > float"(A=Vector Length(A=Get PawnVelocity()), B=10)
   > n7 = Set Movement_alpha_interpolation(Movement_alpha_interpolation:double=FInterpTo(Current=Get Movement_alpha_interpolation(), Target="To Float (Boolean)"(InBool=@n6), DeltaTime=@n1.DeltaTimeX, InterpSpeed=10)) -> (Output_Get:double)
   n8 = GetSocketTransform!(self=Get Body_skelMesh() -> (Body_skelMesh:Skeletal Mesh Component Object Reference), InSocketName=thigh_l, TransformSpace=RTS_Component) -> (
    ReturnValue_Location:Vector
    ReturnValue_Rotation_Roll:float
    ReturnValue_Rotation_Pitch:float
    ReturnValue_Rotation_Yaw:float
    ReturnValue_Scale:Vector
   )
   > Set Thigh_L_crouch_alpha_RESERV(Thigh_L_crouch_alpha_RESERV:double=float * float(A=MapRangeClamped(Value=@n8.ReturnValue_Rotation_Pitch, InRangeA=20, InRangeB=85, OutRangeA=1, OutRangeB=0), B=@n7) {is=title K2Node_PromotableOperator}) -> (Output_Get:double)
   > Set Thigh_l_alpha(Thigh_l_alpha:double=MapRangeClamped(Value=@n8.ReturnValue_Rotation_Pitch, InRangeA=44, InRangeB=64, OutRangeA=1, OutRangeB=0)) -> (Output_Get:double)
   > Set Thigh_L_transform_crouching(Thigh_L_transform_crouching:Vector=MakeVector(X=-14, Y=MapRangeClamped(Value=@n8.ReturnValue_Rotation_Roll, InRangeA=-50, InRangeB=-134, OutRangeA=14, OutRangeB=-8))) -> (Output_Get:Vector)
   n9 = GetSocketTransform(self=Get Body_skelMesh(), InSocketName=thigh_r, TransformSpace=RTS_Component)
   > Set Thigh_r_alpha(Thigh_r_alpha:double=MapRangeClamped(Value=@n9.ReturnValue_Rotation_Pitch, InRangeA=-30, InRangeB=-70, OutRangeA=1, OutRangeB=0)) -> (Output_Get:double)
   > Set Thigh_r_crouch_alpha(Thigh_r_crouch_alpha:double=MapRangeClamped(Value=@n9.ReturnValue_Rotation_Pitch, InRangeA=-23, InRangeB=-63, OutRangeA=1, OutRangeB=0)) -> (Output_Get:double)
   n10 = NOT Boolean(A=Get bIsSkyDiving())
   > Set Skydiving_alpha(Skydiving_alpha:double=MapRangeClamped(Value="To Float (Boolean)"(InBool=@n10), InRangeA=0, InRangeB=1, OutRangeA=0.5)) -> (Output_Get:double)
   Event BlueprintInitializeAnimation()
   > n11 = Cast To PlayerPawn_Athena(Object=TryGetPawnOwner())
   > Set Body_skelMesh(Body_skelMesh:Skeletal Mesh Component Object Reference=GetSkeletalMeshForPartType(self=@n11)) -> (Output_Get:Skeletal Mesh Component Object Reference)
   Event OnCharacterCustomizationLoaded()
   > ResetDynamics(InTeleportType=TeleportPhysics)
   > ResetDynamics(InTeleportType=ResetPhysics)
  ]}
]
variables [
 Body_skelMesh {type=SkeletalMeshComponent}
 Movement_alpha_interpolation {type=number}
 Skydiving_alpha {type=number}
 Thigh_l_alpha {type=number, default=1}
 Thigh_L_crouch_alpha_RESERV {type=number}
 Thigh_L_transform_crouching {type=Vector}
 Thigh_r_alpha {type=number, default=1}
 Thigh_r_crouch_alpha {type=number, default=1}
 windVector {type=Vector, default=<0 0 -980>}
]
