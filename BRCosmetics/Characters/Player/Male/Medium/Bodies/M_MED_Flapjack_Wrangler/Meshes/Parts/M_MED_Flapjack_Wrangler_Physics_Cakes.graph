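# Physics asset for the pancake stack and butter part of the male medium Flapjack Wrangler character cosmetic, defining jiggle bodies for the dyn_ bones.
# Capsule bodies cover pancake bases B to G, their side bones and the butter, with E_side_01 notably larger than the other sides.
# Each side is constrained to its own base with locked twist and mostly single-axis swing, while butter has tight limits against base G.
# No constraints link the bases to each other.
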PhysicsAsset /BRCosmetics/Characters/Player/Male/Medium/Bodies/M_MED_Flapjack_Wrangler/Meshes/Parts/M_MED_Flapjack_Wrangler_Physics_Cakes
bodies {
 dyn_butter [{shapeType=Capsule, length=2, radius=1, center=<0.5 0 0>}]
 dyn_pancake_B_side_02 [{shapeType=Capsule, length=2, radius=1, center=<0.9 0 0>, rotation=<rot 90 0 0>}]
 dyn_pancake_C_side_02 [{shapeType=Capsule, length=2, radius=1, center=<0.9 0 0>, rotation=<rot 90 0 0>}]
 dyn_pancake_D_side_01 [{shapeType=Capsule, length=2, radius=1, center=<0.9 0 0>, rotation=<rot 90 0 0>}]
 dyn_pancake_E_side_01 [{shapeType=Capsule, length=6.3772, radius=4.8522, center=<9.8112 0 0>, rotation=<rot 90 0 0>}]
 dyn_pancake_E_side_02 [{shapeType=Capsule, length=2, radius=1, center=<0.9 0 0>, rotation=<rot 90 0 0>}]
 dyn_pancake_F_side_01 [{shapeType=Capsule, length=2, radius=1, center=<0.9 0 0>, rotation=<rot 90 0 0>}]
 dyn_pancake_F_side_02 [{shapeType=Capsule, length=2, radius=1, center=<0.9 0 0>, rotation=<rot 90 0 0>}]
 dyn_pancake_base_B [{shapeType=Capsule, length=3, radius=1, rotation=<rot 90 0 0>}]
 dyn_pancake_base_C [{shapeType=Capsule, length=3, radius=1, rotation=<rot 90 0 0>}]
 dyn_pancake_base_D [{shapeType=Capsule, length=3, radius=1, rotation=<rot 90 0 0>}]
 dyn_pancake_base_E [{shapeType=Capsule, length=3, radius=1, rotation=<rot 90 0 0>}]
 dyn_pancake_base_F [{shapeType=Capsule, length=3, radius=1, rotation=<rot 90 0 0>}]
 dyn_pancake_base_G [{shapeType=Capsule, length=3, radius=1, center=<-1.8491 -2.004 -3.352>, rotation=<rot 90 0 0>}]
}
constraints [
 {bone1Name=dyn_butter, bone2Name=dyn_pancake_base_G, swing1LimitDegrees=2, swing2LimitDegrees=2, twistLimitDegrees=2}
 {
  bone1Name=dyn_pancake_F_side_01
  bone2Name=dyn_pancake_base_F
  swing1LimitDegrees=0.1
  swing2LimitDegrees=10
  twistLimitDegrees=15
  twistMotion=Locked
 }
 {
  bone1Name=dyn_pancake_F_side_02
  bone2Name=dyn_pancake_base_F
  swing1LimitDegrees=0.1
  swing2LimitDegrees=10
  twistLimitDegrees=15
  twistMotion=Locked
 }
 {
  bone1Name=dyn_pancake_E_side_01
  bone2Name=dyn_pancake_base_E
  swing1LimitDegrees=0.1
  swing2LimitDegrees=10
  twistLimitDegrees=15
  twistMotion=Locked
 }
 {
  bone1Name=dyn_pancake_E_side_02
  bone2Name=dyn_pancake_base_E
  swing1LimitDegrees=0.1
  swing2LimitDegrees=10
  twistLimitDegrees=15
  twistMotion=Locked
 }
 {
  bone1Name=dyn_pancake_D_side_01
  bone2Name=dyn_pancake_base_D
  swing1LimitDegrees=0.1
  swing2LimitDegrees=10
  twistLimitDegrees=15
  twistMotion=Locked
 }
 {
  bone1Name=dyn_pancake_C_side_02
  bone2Name=dyn_pancake_base_C
  swing1LimitDegrees=0.1
  swing2LimitDegrees=10
  twistLimitDegrees=15
  twistMotion=Locked
 }
 {
  bone1Name=dyn_pancake_B_side_02
  bone2Name=dyn_pancake_base_B
  swing1LimitDegrees=0.1
  swing2LimitDegrees=10
  twistLimitDegrees=15
  twistMotion=Locked
 }
]
