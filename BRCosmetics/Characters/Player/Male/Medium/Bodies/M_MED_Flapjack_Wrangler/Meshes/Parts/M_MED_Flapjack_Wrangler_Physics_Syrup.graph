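# Physics asset for the syrup part of the Flapjack Wrangler male cosmetic body, simulating dangling syrup on a pancake stack.
# Small capsules cover the pancake and syrup bones, with one large capsule on the head.
# Two twist-locked side bones hang off the pancake G base, and one- or two-bone syrup chains attach to them or directly to that base.
# The head and pancake bases B through F have bodies but no constraints.
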PhysicsAsset /BRCosmetics/Characters/Player/Male/Medium/Bodies/M_MED_Flapjack_Wrangler/Meshes/Parts/M_MED_Flapjack_Wrangler_Physics_Syrup
bodies {
 dyn_pancake_G_side_01 [{shapeType=Capsule, length=2, radius=1, center=<0.9 0 0>, rotation=<rot 90 0 0>}]
 dyn_pancake_G_side_02 [{shapeType=Capsule, length=2, radius=1, center=<0.9 0 0>, rotation=<rot 90 0 0>}]
 dyn_pancake_base_B [{shapeType=Capsule, length=3, radius=1, rotation=<rot 90 0 0>}]
 dyn_pancake_base_C [{shapeType=Capsule, length=3, radius=1, rotation=<rot 90 0 0>}]
 dyn_pancake_base_D [{shapeType=Capsule, length=3, radius=1, rotation=<rot 90 0 0>}]
 dyn_pancake_base_E [{shapeType=Capsule, length=3, radius=1, rotation=<rot 90 0 0>}]
 dyn_pancake_base_F [{shapeType=Capsule, length=3, radius=1, rotation=<rot 90 0 0>}]
 dyn_pancake_base_G [{shapeType=Capsule, length=3, radius=1, center=<-1.8491 -2.004 -3.352>, rotation=<rot 90 0 0>}]
 dyn_syrup_A_01 [{shapeType=Capsule, length=3, radius=1, center=<3.097 -0.1231 -0.0125>, rotation=<rot 90 0 0>}]
 dyn_syrup_A_02 [{shapeType=Capsule, length=3, radius=1, center=<0.6479 -0.0302 0.0098>, rotation=<rot 90 0 0>}]
 dyn_syrup_B_01 [{shapeType=Capsule, length=3, radius=1, center=<2.8323 0.0592 0.1887>, rotation=<rot 90 0 0>}]
 dyn_syrup_C_01 [{shapeType=Capsule, length=3, radius=1, center=<2.996 -0.2092 -0.0715>, rotation=<rot 90 0 0>}]
 dyn_syrup_C_02 [{shapeType=Capsule, length=3, radius=1, center=<1.8596 -0.0064 -0.0242>, rotation=<rot 90 0 0>}]
 dyn_syrup_D_01 [{shapeType=Capsule, length=3, radius=1, center=<2.4558 0.0285 0.0184>, rotation=<rot 90 0 0>}]
 dyn_syrup_E_01 [{shapeType=Capsule, length=3, radius=1, center=<2.4778 -0.109 0.0581>, rotation=<rot 90 0 0>}]
 dyn_syrup_E_02 [{shapeType=Capsule, length=3, radius=1, center=<1.9535 0.0392 0.0698>, rotation=<rot 90 0 0>}]
 dyn_syrup_F_01 [{shapeType=Capsule, length=3, radius=1, center=<2.669 0.1727 -0.1462>, rotation=<rot 90 0 0>}]
 dyn_syrup_F_02 [{shapeType=Capsule, length=3, radius=1, center=<1.5576 -0.0858 -0.0415>, rotation=<rot 90 0 0>}]
 dyn_syrup_G_01 [{shapeType=Capsule, length=3, radius=1, center=<2.2252 -0.0001 -0.2818>, rotation=<rot 90 0 0>}]
 dyn_syrup_G_02 [{shapeType=Capsule, length=3, radius=1, center=<1.8667 0.1203 -0.3163>, rotation=<rot 90 0 0>}]
 dyn_syrup_H_01 [{shapeType=Capsule, length=3, radius=1, center=<1.4292 0 0>, rotation=<rot 90 0 0>}]
 dyn_syrup_H_02 [{shapeType=Capsule, length=3, radius=1, center=<1.4998 0.0285 -0.0744>, rotation=<rot 90 0 0>}]
 dyn_syrup_I_01 [{shapeType=Capsule, length=3, radius=1, center=<2.1413 -0.0073 -0.0769>, rotation=<rot 90 0 0>}]
 dyn_syrup_I_02 [{shapeType=Capsule, length=3, radius=1, center=<0.3374 -0.1622 0.091>, rotation=<rot 90 0 0>}]
 dyn_syrup_J_01 [{shapeType=Capsule, length=3, radius=1, center=<1.9376 0.0604 -0.144>, rotation=<rot 90 0 0>}]
 dyn_syrup_K_01 [{shapeType=Capsule, length=3, radius=1, center=<2.1631 -0.0578 0.0558>, rotation=<rot 90 0 0>}]
 dyn_syrup_L_01 [{shapeType=Capsule, length=3, radius=1, center=<2.0252 -0.0184 0.0712>, rotation=<rot 90 0 0>}]
 head [{shapeType=Capsule, length=3.3, radius=10, center=<0 1.7 -0.5>, rotation=<rot 90 0 0>}]
}
constraints [
 {
  bone1Name=dyn_pancake_G_side_01
  bone2Name=dyn_pancake_base_G
  swing1LimitDegrees=0.1
  swing2LimitDegrees=10
  twistLimitDegrees=15
  twistMotion=Locked
 }
 {bone1Name=dyn_syrup_A_01, bone2Name=dyn_pancake_G_side_01, swing1LimitDegrees=30, swing2LimitDegrees=30, twistLimitDegrees=5}
 {bone1Name=dyn_syrup_A_02, bone2Name=dyn_syrup_A_01, swing1LimitDegrees=30, swing2LimitDegrees=30, twistLimitDegrees=5}
 {bone1Name=dyn_syrup_J_01, bone2Name=dyn_pancake_G_side_01, swing1LimitDegrees=30, swing2LimitDegrees=30, twistLimitDegrees=5}
 {bone1Name=dyn_syrup_B_01, bone2Name=dyn_pancake_G_side_01, swing1LimitDegrees=30, swing2LimitDegrees=30, twistLimitDegrees=5}
 {
  bone1Name=dyn_pancake_G_side_02
  bone2Name=dyn_pancake_base_G
  swing1LimitDegrees=0.1
  swing2LimitDegrees=10
  twistLimitDegrees=15
  twistMotion=Locked
 }
 {bone1Name=dyn_syrup_D_01, bone2Name=dyn_pancake_G_side_02, swing1LimitDegrees=30, swing2LimitDegrees=30, twistLimitDegrees=5}
 {bone1Name=dyn_syrup_E_01, bone2Name=dyn_pancake_G_side_02, swing1LimitDegrees=30, swing2LimitDegrees=30, twistLimitDegrees=5}
 {bone1Name=dyn_syrup_E_02, bone2Name=dyn_syrup_E_01, swing1LimitDegrees=30, swing2LimitDegrees=30, twistLimitDegrees=5}
 {bone1Name=dyn_syrup_F_01, bone2Name=dyn_pancake_G_side_02, swing1LimitDegrees=30, swing2LimitDegrees=30, twistLimitDegrees=5}
 {bone1Name=dyn_syrup_F_02, bone2Name=dyn_syrup_F_01, swing1LimitDegrees=30, swing2LimitDegrees=30, twistLimitDegrees=5}
 {bone1Name=dyn_syrup_C_01, bone2Name=dyn_pancake_base_G, swing1LimitDegrees=30, swing2LimitDegrees=30, twistLimitDegrees=5}
 {bone1Name=dyn_syrup_C_02, bone2Name=dyn_syrup_C_01, swing1LimitDegrees=30, swing2LimitDegrees=30, twistLimitDegrees=5}
 {bone1Name=dyn_syrup_I_01, bone2Name=dyn_pancake_base_G, swing1LimitDegrees=30, swing2LimitDegrees=30, twistLimitDegrees=5}
 {bone1Name=dyn_syrup_I_02, bone2Name=dyn_syrup_I_01, swing1LimitDegrees=30, swing2LimitDegrees=30, twistLimitDegrees=5}
 {bone1Name=dyn_syrup_G_01, bone2Name=dyn_pancake_G_side_02, swing1LimitDegrees=30, swing2LimitDegrees=30, twistLimitDegrees=5}
 {bone1Name=dyn_syrup_G_02, bone2Name=dyn_syrup_G_01, swing1LimitDegrees=30, swing2LimitDegrees=30, twistLimitDegrees=5}
 {bone1Name=dyn_syrup_H_01, bone2Name=dyn_pancake_G_side_02, swing1LimitDegrees=30, swing2LimitDegrees=30, twistLimitDegrees=5}
 {bone1Name=dyn_syrup_H_02, bone2Name=dyn_syrup_H_01, swing1LimitDegrees=30, swing2LimitDegrees=30, twistLimitDegrees=5}
 {bone1Name=dyn_syrup_K_01, bone2Name=dyn_pancake_G_side_02, swing1LimitDegrees=30, swing2LimitDegrees=30, twistLimitDegrees=5}
 {bone1Name=dyn_syrup_L_01, bone2Name=dyn_pancake_G_side_02, swing1LimitDegrees=30, swing2LimitDegrees=30, twistLimitDegrees=5}
]
