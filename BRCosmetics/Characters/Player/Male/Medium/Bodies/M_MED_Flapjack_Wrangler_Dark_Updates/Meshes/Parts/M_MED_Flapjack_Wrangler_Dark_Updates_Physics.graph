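# Physics asset for the Flapjack Wrangler Dark male medium character body, covering the dynamic hat bones for secondary motion.
# Four identical thin box bodies sit on the front, back, left and right hat bones.
# Each is constrained directly to a small box on head_ALT_G with tight swing and twist limits, with no links between hat sections.
# A tiny offset box on the head bone has no constraint.
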PhysicsAsset /BRCosmetics/Characters/Player/Male/Medium/Bodies/M_MED_Flapjack_Wrangler_Dark_Updates/Meshes/Parts/M_MED_Flapjack_Wrangler_Dark_Updates_Physics
bodies {
 dyn_hatBack [{shapeType=Box, extentX=6, extentY=15, extentZ=1, center=<3 0 0>}]
 dyn_hatFront [{shapeType=Box, extentX=6, extentY=15, extentZ=1, center=<3 0 0>}]
 dyn_hatLeft [{shapeType=Box, extentX=6, extentY=15, extentZ=1, center=<3 0 0>}]
 dyn_hatRight [{shapeType=Box, extentX=6, extentY=15, extentZ=1, center=<3 0 0>}]
 head [{shapeType=Box, extentX=0.5, extentY=0.5, extentZ=0.5, center=<10.4508 3.2796 -0.001>}]
 head_ALT_G [{shapeType=Box, extentX=1, extentY=1, extentZ=1}]
}
constraints [
 {bone1Name=dyn_hatFront, bone2Name=head_ALT_G, swing1LimitDegrees=10, swing2LimitDegrees=15, twistLimitDegrees=5}
 {bone1Name=dyn_hatBack, bone2Name=head_ALT_G, swing1LimitDegrees=10, swing2LimitDegrees=15, twistLimitDegrees=5}
 {bone1Name=dyn_hatLeft, bone2Name=head_ALT_G, swing1LimitDegrees=10, swing2LimitDegrees=15, twistLimitDegrees=5}
 {bone1Name=dyn_hatRight, bone2Name=head_ALT_G, swing1LimitDegrees=10, swing2LimitDegrees=15, twistLimitDegrees=5}
]
