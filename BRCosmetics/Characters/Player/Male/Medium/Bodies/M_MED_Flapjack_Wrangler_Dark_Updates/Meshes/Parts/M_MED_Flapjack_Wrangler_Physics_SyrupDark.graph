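# Physics asset for the dark-syrup Flapjack Wrangler cosmetic, giving secondary motion to its pancake stack, butter, syrup drips and hat.
# Capsules cover the dyn_ bones and upper spine, neck and head, with free constraints between body bones and a small box on head_ALT_G.
# Pancake bases lock to neck or head bones and carry tightly limited side pieces.
# Pancake G carries butter and one- or two-link syrup chains, while head_ALT_G anchors it, the hat and hat syrup.
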PhysicsAsset /BRCosmetics/Characters/Player/Male/Medium/Bodies/M_MED_Flapjack_Wrangler_Dark_Updates/Meshes/Parts/M_MED_Flapjack_Wrangler_Physics_SyrupDark
bodies {
 dyn_butter [{shapeType=Capsule, length=2, radius=1, center=<0.5 0 0>}]
 dyn_hatBack [{shapeType=Capsule, length=3, radius=3.6, shapeName=dyn_hatBack_capsule, center=<3 0 0>, rotation=<rot 0 0 90>}]
 dyn_hatFront [{shapeType=Capsule, length=3, radius=3.6, shapeName=dyn_hatFront_capsule, center=<3 0 0>, rotation=<rot 0 0 90>}]
 dyn_hatLeft [{shapeType=Capsule, length=3, radius=3.6, shapeName=dyn_hatLeft_capsule, center=<4.6 0 0>, rotation=<rot 0 0 90>}]
 dyn_hatRight [{shapeType=Capsule, length=3, radius=3.6, shapeName=dyn_hatRight_capsule, center=<4.6 0 0>, rotation=<rot 0 0 90>}]
 dyn_pancake_B_side_02 [{shapeType=Capsule, length=2, radius=1, center=<0.9 0 0>, rotation=<rot 90 0 0>}]
 dyn_pancake_C_side_02 [{shapeType=Capsule, length=2, radius=1, center=<0.9 0 0>, rotation=<rot 90 0 0>}]
 dyn_pancake_D_side_01 [{shapeType=Capsule, length=2, radius=1, center=<0.9 0 0>, rotation=<rot 90 0 0>}]
 dyn_pancake_E_side_01 [{shapeType=Capsule, length=2, radius=1, center=<0.9 0 0>, rotation=<rot 90 0 0>}]
 dyn_pancake_E_side_02 [{shapeType=Capsule, length=2, radius=1, center=<0.9 0 0>, rotation=<rot 90 0 0>}]
 dyn_pancake_F_side_01 [{shapeType=Capsule, length=2, radius=1, center=<0.9 0 0>, rotation=<rot 90 0 0>}]
 dyn_pancake_F_side_02 [{shapeType=Capsule, length=2, radius=1, center=<0.9 0 0>, rotation=<rot 90 0 0>}]
 dyn_pancake_G_side_01 [{shapeType=Capsule, length=2, radius=1, center=<0.9 0 0>, rotation=<rot 90 0 0>}]
 dyn_pancake_G_side_02 [{shapeType=Capsule, length=2, radius=1, center=<0.9 0 0>, rotation=<rot 90 0 0>}]
 dyn_pancake_base_B [{shapeType=Capsule, length=3, radius=1, rotation=<rot 90 0 0>}]
 dyn_pancake_base_C [{shapeType=Capsule, length=3, radius=1, rotation=<rot 90 0 0>}]
 dyn_pancake_base_D [{shapeType=Capsule, length=3, radius=1, rotation=<rot 90 0 0>}]
 dyn_pancake_base_E [{shapeType=Capsule, length=3, radius=1, rotation=<rot 90 0 0>}]
 dyn_pancake_base_F [{shapeType=Capsule, length=3, radius=1, rotation=<rot 90 0 0>}]
 dyn_pancake_base_G [{shapeType=Capsule, length=3, radius=1, center=<-1.8491 -2.004 -3.352>, rotation=<rot 90 0 0>}]
 dyn_syrupHat_A_01 [{shapeType=Capsule, length=3, radius=1, center=<2 -0.1328 0.0545>, rotation=<rot 90 0 0>}]
 dyn_syrupHat_B_01 [{shapeType=Capsule, length=3, radius=1, center=<2 -0.011 -0.1458>, rotation=<rot 90 0 0>}]
 dyn_syrupHat_C_01 [{shapeType=Capsule, length=3, radius=1, center=<2 -0.049 -0.052>, rotation=<rot 90 0 0>}]
 dyn_syrupHat_D_01 [{shapeType=Capsule, length=3, radius=1, center=<2 0 0>, rotation=<rot 90 0 0>}]
 dyn_syrupHat_E_01 [{shapeType=Capsule, length=3, radius=1, center=<2 0.4221 -0.0741>, rotation=<rot 90 0 0>}]
 dyn_syrup_A_01 [{shapeType=Capsule, length=3, radius=1, center=<3.097 -0.1231 -0.0125>, rotation=<rot 90 0 0>}]
 dyn_syrup_A_02 [{shapeType=Capsule, length=3, radius=1, center=<0.6479 -0.0302 0.0098>, rotation=<rot 90 0 0>}]
 dyn_syrup_B_01 [{shapeType=Capsule, length=3, radius=1, center=<2.8323 0.0592 0.1887>, rotation=<rot 90 0 0>}]
 dyn_syrup_C_01 [{shapeType=Capsule, length=3, radius=1, center=<2.996 -0.2092 -0.0715>, rotation=<rot 90 0 0>}]
 dyn_syrup_C_02 [{shapeType=Capsule, length=3, radius=1, center=<1.8596 -0.0064 -0.0242>, rotation=<rot 90 0 0>}]
 dyn_syrup_D_01 [{shapeType=Capsule, length=3, radius=1, center=<2.4558 0.0285 0.0184>, rotation=<rot 90 0 0>}]
 dyn_syrup_E_01 [{shapeType=Capsule, length=3, radius=1, center=<2.4778 -0.109 0.0581>, rotation=<rot 90 0 0>}]
 dyn_syrup_E_02 [{shapeType=Capsule, length=3, radius=1, center=<1.9535 0.0392 0.0698>, rotation=<rot 90 0 0>}]
 dyn_syrup_F_01 [{shapeType=Capsule, length=3, radius=1, center=<2.669 0.1727 -0.1462>, rotation=<rot 90 0 0>}]
 dyn_syrup_F_02 [{shapeType=Capsule, length=3, radius=1, center=<1.5576 -0.0858 -0.0415>, rotation=<rot 90 0 0>}]
 dyn_syrup_G_01 [{shapeType=Capsule, length=3, radius=1, center=<2.2252 -0.0001 -0.2818>, rotation=<rot 90 0 0>}]
 dyn_syrup_G_02 [{shapeType=Capsule, length=3, radius=1, center=<1.8667 0.1203 -0.3163>, rotation=<rot 90 0 0>}]
 dyn_syrup_H_01 [{shapeType=Capsule, length=3, radius=1, center=<1.4292 0 0>, rotation=<rot 90 0 0>}]
 dyn_syrup_H_02 [{shapeType=Capsule, length=3, radius=1, center=<1.5038 0.014 -0.0828>, rotation=<rot 90 0 0>}]
 dyn_syrup_I_01 [{shapeType=Capsule, length=3, radius=1, center=<2.1413 -0.0073 -0.0769>, rotation=<rot 90 0 0>}]
 dyn_syrup_I_02 [{shapeType=Capsule, length=3, radius=1, center=<0.3374 -0.1622 0.091>, rotation=<rot 90 0 0>}]
 dyn_syrup_J_01 [{shapeType=Capsule, length=3, radius=1, center=<1.9376 0.0604 -0.144>, rotation=<rot 90 0 0>}]
 dyn_syrup_K_01 [{shapeType=Capsule, length=3, radius=1, center=<2.1631 -0.0578 0.0558>, rotation=<rot 90 0 0>}]
 dyn_syrup_L_01 [{shapeType=Capsule, length=3, radius=1, center=<2.0252 0 0.0712>, rotation=<rot 90 0 0>}]
 head [{shapeType=Capsule, length=3.3, radius=10, center=<0 1.7 -0.5>, rotation=<rot 90 0 0>}]
 head_ALT_G [{shapeType=Box, extentX=1, extentY=1, extentZ=1}]
 neck_01 [{shapeType=Capsule, length=3.3, radius=10, center=<0 3.5 -0.5>, rotation=<rot 90 0 0>}]
 neck_02 [{shapeType=Capsule, length=3.3, radius=10, center=<0 2.9 -0.5>, rotation=<rot 90 0 10>}]
 spine_05 [{shapeType=Capsule, length=17, radius=11.3, center=<0 3.5 -0.5>}]
}
constraints [
 {
  bone1Name=dyn_pancake_G_side_01
  bone2Name=dyn_pancake_base_G
  swing1LimitDegrees=45
  swing1Motion=Locked
  swing2LimitDegrees=10
  twistLimitDegrees=15
  twistMotion=Locked
 }
 {bone1Name=dyn_syrup_A_01, bone2Name=dyn_pancake_G_side_01, swing1LimitDegrees=30, swing2LimitDegrees=30, twistLimitDegrees=5}
 {bone1Name=dyn_syrup_A_02, bone2Name=dyn_syrup_A_01, swing1LimitDegrees=30, swing2LimitDegrees=30, twistLimitDegrees=5}
 {bone1Name=dyn_syrup_J_01, bone2Name=dyn_pancake_G_side_01, swing1LimitDegrees=30, swing2LimitDegrees=30, twistLimitDegrees=5}
 {bone1Name=dyn_syrup_B_01, bone2Name=dyn_pancake_G_side_01, swing1LimitDegrees=30, swing2LimitDegrees=30, twistLimitDegrees=5}
 {
  bone1Name=dyn_pancake_G_side_02
  bone2Name=dyn_pancake_base_G
  swing1LimitDegrees=45
  swing1Motion=Locked
  swing2LimitDegrees=10
  twistLimitDegrees=15
  twistMotion=Locked
 }
 {bone1Name=dyn_syrup_D_01, bone2Name=dyn_pancake_G_side_02, swing1LimitDegrees=30, swing2LimitDegrees=30, twistLimitDegrees=5}
 {bone1Name=dyn_syrup_E_01, bone2Name=dyn_pancake_G_side_02, swing1LimitDegrees=30, swing2LimitDegrees=30, twistLimitDegrees=5}
 {bone1Name=dyn_syrup_E_02, bone2Name=dyn_syrup_E_01, swing1LimitDegrees=30, swing2LimitDegrees=30, twistLimitDegrees=5}
 {bone1Name=dyn_syrup_F_01, bone2Name=dyn_pancake_G_side_02, swing1LimitDegrees=30, swing2LimitDegrees=30, twistLimitDegrees=5}
 {bone1Name=dyn_syrup_F_02, bone2Name=dyn_syrup_F_01, swing1LimitDegrees=30, swing2LimitDegrees=30, twistLimitDegrees=5}
 {bone1Name=dyn_syrup_C_01, bone2Name=dyn_pancake_base_G, swing1LimitDegrees=30, swing2LimitDegrees=30, twistLimitDegrees=5}
 {bone1Name=dyn_syrup_C_02, bone2Name=dyn_syrup_C_01, swing1LimitDegrees=30, swing2LimitDegrees=30, twistLimitDegrees=5}
 {bone1Name=dyn_syrup_I_01, bone2Name=dyn_pancake_base_G, swing1LimitDegrees=30, swing2LimitDegrees=30, twistLimitDegrees=5}
 {bone1Name=dyn_syrup_I_02, bone2Name=dyn_syrup_I_01, swing1LimitDegrees=30, swing2LimitDegrees=30, twistLimitDegrees=5}
 {bone1Name=dyn_butter, bone2Name=dyn_pancake_base_G, swing1LimitDegrees=2, swing2LimitDegrees=2, twistLimitDegrees=2}
 {
  bone1Name=dyn_pancake_F_side_01
  bone2Name=dyn_pancake_base_F
  swing1LimitDegrees=0.1
  swing2LimitDegrees=10
  twistLimitDegrees=15
  twistMotion=Locked
 }
 {
  bone1Name=dyn_pancake_F_side_02
  bone2Name=dyn_pancake_base_F
  swing1LimitDegrees=0.1
  swing2LimitDegrees=10
  twistLimitDegrees=15
  twistMotion=Locked
 }
 {
  bone1Name=dyn_pancake_E_side_01
  bone2Name=dyn_pancake_base_E
  swing1LimitDegrees=0.1
  swing2LimitDegrees=10
  twistLimitDegrees=15
  twistMotion=Locked
 }
 {
  bone1Name=dyn_pancake_E_side_02
  bone2Name=dyn_pancake_base_E
  swing1LimitDegrees=0.1
  swing2LimitDegrees=10
  twistLimitDegrees=15
  twistMotion=Locked
 }
 {
  bone1Name=dyn_pancake_D_side_01
  bone2Name=dyn_pancake_base_D
  swing1LimitDegrees=0.1
  swing2LimitDegrees=10
  twistLimitDegrees=15
  twistMotion=Locked
 }
 {
  bone1Name=dyn_pancake_C_side_02
  bone2Name=dyn_pancake_base_C
  swing1LimitDegrees=0.1
  swing2LimitDegrees=10
  twistLimitDegrees=15
  twistMotion=Locked
 }
 {
  bone1Name=dyn_pancake_B_side_02
  bone2Name=dyn_pancake_base_B
  swing1LimitDegrees=0.1
  swing2LimitDegrees=10
  twistLimitDegrees=15
  twistMotion=Locked
 }
 {
  bone1Name=head
  bone2Name=neck_02
  swing1LimitDegrees=45
  swing1Motion=Free
  swing2LimitDegrees=45
  swing2Motion=Free
  twistLimitDegrees=45
  twistMotion=Free
 }
 {
  bone1Name=neck_02
  bone2Name=neck_01
  swing1LimitDegrees=45
  swing1Motion=Free
  swing2LimitDegrees=45
  swing2Motion=Free
  twistLimitDegrees=45
  twistMotion=Free
 }
 {
  bone1Name=neck_01
  bone2Name=spine_05
  swing1LimitDegrees=45
  swing1Motion=Free
  swing2LimitDegrees=45
  swing2Motion=Free
  twistLimitDegrees=45
  twistMotion=Free
 }
 {
  bone1Name=dyn_pancake_base_G
  bone2Name=head_ALT_G
  swing1LimitDegrees=45
  swing1Motion=Locked
  swing2LimitDegrees=45
  swing2Motion=Locked
  twistLimitDegrees=15
  twistMotion=Locked
 }
 {
  bone1Name=dyn_pancake_base_F
  bone2Name=head
  swing1LimitDegrees=45
  swing1Motion=Locked
  swing2LimitDegrees=45
  swing2Motion=Locked
  twistLimitDegrees=15
  twistMotion=Locked
 }
 {
  bone1Name=dyn_pancake_base_E
  bone2Name=head
  swing1LimitDegrees=45
  swing1Motion=Locked
  swing2LimitDegrees=45
  swing2Motion=Locked
  twistLimitDegrees=15
  twistMotion=Locked
 }
 {
  bone1Name=dyn_pancake_base_D
  bone2Name=neck_02
  swing1LimitDegrees=45
  swing1Motion=Locked
  swing2LimitDegrees=45
  swing2Motion=Locked
  twistLimitDegrees=15
  twistMotion=Locked
 }
 {
  bone1Name=dyn_pancake_base_C
  bone2Name=neck_02
  swing1LimitDegrees=45
  swing1Motion=Locked
  swing2LimitDegrees=45
  swing2Motion=Locked
  twistLimitDegrees=15
  twistMotion=Locked
 }
 {
  bone1Name=dyn_pancake_base_B
  bone2Name=neck_01
  swing1LimitDegrees=45
  swing1Motion=Locked
  swing2LimitDegrees=45
  swing2Motion=Locked
  twistLimitDegrees=15
  twistMotion=Locked
 }
 {bone1Name=dyn_hatFront, bone2Name=head_ALT_G, swing1LimitDegrees=10, swing2LimitDegrees=15, twistLimitDegrees=5}
 {bone1Name=dyn_hatBack, bone2Name=head_ALT_G, swing1LimitDegrees=10, swing2LimitDegrees=15, twistLimitDegrees=5}
 {bone1Name=dyn_hatLeft, bone2Name=head_ALT_G, swing1LimitDegrees=10, swing2LimitDegrees=15, twistLimitDegrees=5}
 {bone1Name=dyn_hatRight, bone2Name=head_ALT_G, swing1LimitDegrees=10, swing2LimitDegrees=15, twistLimitDegrees=5}
 {
  bone1Name=head_ALT_G
  bone2Name=spine_05
  swing1LimitDegrees=45
  swing1Motion=Free
  swing2LimitDegrees=45
  swing2Motion=Free
  twistLimitDegrees=45
  twistMotion=Free
 }
 {bone1Name=dyn_syrup_H_01, bone2Name=dyn_pancake_G_side_02, swing1LimitDegrees=30, swing2LimitDegrees=30, twistLimitDegrees=5}
 {bone1Name=dyn_syrup_H_02, bone2Name=dyn_syrup_H_01, swing1LimitDegrees=30, swing2LimitDegrees=30, twistLimitDegrees=5}
 {bone1Name=dyn_syrup_G_01, bone2Name=dyn_pancake_G_side_02, swing1LimitDegrees=30, swing2LimitDegrees=30, twistLimitDegrees=5}
 {bone1Name=dyn_syrup_G_02, bone2Name=dyn_syrup_G_01, swing1LimitDegrees=30, swing2LimitDegrees=30, twistLimitDegrees=5}
 {bone1Name=dyn_syrup_L_01, bone2Name=dyn_pancake_G_side_02, swing1LimitDegrees=30, swing2LimitDegrees=30, twistLimitDegrees=5}
 {bone1Name=dyn_syrup_K_01, bone2Name=dyn_pancake_G_side_02, swing1LimitDegrees=30, swing2LimitDegrees=30, twistLimitDegrees=5}
 {bone1Name=dyn_syrupHat_A_01, bone2Name=head_ALT_G, swing1LimitDegrees=25, swing2LimitDegrees=2, swing2Motion=Locked, twistLimitDegrees=5}
 {bone1Name=dyn_syrupHat_B_01, bone2Name=head_ALT_G, swing1LimitDegrees=25, swing2LimitDegrees=2, swing2Motion=Locked, twistLimitDegrees=5}
 {bone1Name=dyn_syrupHat_C_01, bone2Name=head_ALT_G, swing1LimitDegrees=25, swing2LimitDegrees=2, swing2Motion=Locked, twistLimitDegrees=5}
 {bone1Name=dyn_syrupHat_D_01, bone2Name=head_ALT_G, swing1LimitDegrees=25, swing2LimitDegrees=2, swing2Motion=Locked, twistLimitDegrees=5}
 {bone1Name=dyn_syrupHat_E_01, bone2Name=head_ALT_G, swing1LimitDegrees=25, swing2LimitDegrees=2, swing2Motion=Locked, twistLimitDegrees=5}
]
