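# Control Rig for the Flapjack Wrangler player body that carries jaw motion from the face bones to an alternate face on the pancake-stack bones.
# The Jaw control sets C_jaw, whose transform drives C_jaw_ALT_Main through separate rotation and translation constraints with maintained offsets.
# A second chain constrains C_lip_lower_mid_ALT_Main to C_lip_lower_mid the same way.
# The Begin Execution node has no visible connection to either chain.
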ControlRigBlueprint /Game/Characters/Player/Male/Medium/Bodies/M_MED_Flapjack_Wrangler/Meshes/Parts/M_MED_Flapjack_Wrangler_CtrlRig
graphs [
 RigVMFunctionLibrary {links [], nodes []}
 RigVMModel {nodes [
   RigUnit_GetControlTransform = /Script/ControlRig.RigUnit_GetControlTransform(Control=Jaw, Space=GlobalSpace) {is=struct RigVMUnitNode, template="GetControlValue::Execute(in Control,in Space,out Transform,out Minimum,out Maximum)"}
   /Script/ControlRig.RigUnit_SetBoneTransform(Bone=C_jaw, ExecutePin=(), Space=GlobalSpace, Weight=1, bPropagateToChildren=True, Transform=@RigUnit_GetControlTransform.Transform) {is=struct RigVMUnitNode}
   n1 = &RigUnit_GetBoneTransform(Bone=C_jaw)
   > /Script/ControlRig.RigUnit_TransformConstraint(
    BaseBone=None
    BaseTransform=(Rotation=(X=0,Y=0,Z=0,W=1),Translation=(X=0,Y=0,Z=0),Scale3D=(X=1,Y=1,Z=1))
    BaseTransformSpace=GlobalSpace
    Bone=C_jaw_ALT_Main
    Targets="((Transform=(Rotation=(X=0,Y=0,Z=0,W=1),Translation=(X=0,Y=0,Z=0),Scale3D=(X=1,Y=1,Z=1)),Weight=1,bMaintainOffset=True,Filter=(TranslationFilter=(bX=false,bY=false,bZ=false),RotationFilter=(bX=True,bY=True,bZ=True),ScaleFilter=(bX=false,bY=false,bZ=false)) ...(truncated)"
    bUseInitialTransforms=True
    Targets.0.Transform=@n1.Transform
   ) {is=struct RigVMUnitNode}
   > &RigUnit_TransformConstraint(
    BaseBone=None
    BaseTransform=(Rotation=(X=0,Y=0,Z=0,W=1),Translation=(X=0,Y=0,Z=0),Scale3D=(X=1,Y=1,Z=1))
    BaseTransformSpace=GlobalSpace
    Bone=C_jaw_ALT_Main
    Targets="((Transform=(Rotation=(X=0,Y=0,Z=0,W=1),Translation=(X=0,Y=0,Z=0),Scale3D=(X=1,Y=1,Z=1)),Weight=1,bMaintainOffset=True,Filter=(TranslationFilter=(bX=True,bY=True,bZ=True),RotationFilter=(bX=false,bY=false,bZ=false),ScaleFilter=(bX=false,bY=false,bZ=false)) ...(truncated)"
    bUseInitialTransforms=True
    Targets.0.Transform=@n1.Transform
   )
   n2 = &RigUnit_GetBoneTransform(Bone=C_lip_lower_mid)
   &RigUnit_TransformConstraint(
    BaseBone=None
    BaseTransform=(Rotation=(X=0,Y=0,Z=0,W=1),Translation=(X=0,Y=0,Z=0),Scale3D=(X=1,Y=1,Z=1))
    BaseTransformSpace=GlobalSpace
    Bone=C_lip_lower_mid_ALT_Main
    ExecutePin=()
    Targets="((Transform=(Rotation=(X=0,Y=0,Z=0,W=1),Translation=(X=0,Y=0,Z=0),Scale3D=(X=1,Y=1,Z=1)),Weight=1,bMaintainOffset=True,Filter=(TranslationFilter=(bX=false,bY=false,bZ=false),RotationFilter=(bX=True,bY=True,bZ=True),ScaleFilter=(bX=false,bY=false,bZ=false)) ...(truncated)"
    bUseInitialTransforms=True
    Targets.0.Transform=@n2.Transform
   )
   > &RigUnit_TransformConstraint(
    BaseBone=None
    BaseTransform=(Rotation=(X=0,Y=0,Z=0,W=1),Translation=(X=0,Y=0,Z=0),Scale3D=(X=1,Y=1,Z=1))
    BaseTransformSpace=GlobalSpace
    Bone=C_lip_lower_mid_ALT_Main
    Targets="((Transform=(Rotation=(X=0,Y=0,Z=0,W=1),Translation=(X=0,Y=0,Z=0),Scale3D=(X=1,Y=1,Z=1)),Weight=1,bMaintainOffset=True,Filter=(TranslationFilter=(bX=True,bY=True,bZ=True),RotationFilter=(bX=false,bY=false,bZ=false),ScaleFilter=(bX=false,bY=false,bZ=false)) ...(truncated)"
    bUseInitialTransforms=True
    Targets.0.Transform=@n2.Transform
   )
   &RigUnit_BeginExecution()
  ]}
]
hierarchy [
 root {loc=<0 0 0>, rot=<0 0 0 1>}
 pelvis {parent=root, loc=<0 -0.7071 93.1191>, rot=<0.0127 0.707 -0.0127 -0.707>}
 spine_01 {parent=pelvis, loc=<6.7494 -0.6388 0>, rot=<0 0 0.034 -0.9994>}
 spine_02 {parent=spine_01, loc=<7.2518 -0.035 0>, rot=<0 0 0.0238 -0.9997>}
 spine_03 {parent=spine_02, loc=<6.0104 0 0>, rot=<0 0 0.0149 -0.9999>}
 spine_04 {parent=spine_03, loc=<8.1024 0 0>, rot=<0 0 -0.0132 -0.9999>}
 spine_05 {parent=spine_04, loc=<10.0146 0 0>, rot=<0 0 -0.0015 -1>}
 neck_01 {parent=spine_05, loc=<11.3138 0.1529 0>, rot=<0 0 -0.2003 -0.9797>}
 neck_02 {parent=neck_01, loc=<4.5627 0 0>, rot=<0 0 0.0504 -0.9987>}
 head {parent=neck_02, loc=<5.3914 0 0>, rot=<0 0 0.115 -0.9934>}
 faceAttach {parent=head, loc=<-0.8266 -0.1832 0>, rot=<0.0044 -0.3827 -0.0044 -0.9239>, scale=<0.89 0.89 0.89>}
 C_brow_mid {parent=faceAttach, loc=<4.7209 11.2084 4.8303>, rot=<-0.0586 -0.317 0.0114 -0.9466>}
 L_brow_mid {parent=faceAttach, loc=<7.639 10.371 3.157>, rot=<0.0603 -0.3852 0.2489 -0.8866>}
 L_brow_outer {parent=faceAttach, loc=<9.1042 8.1218 1.4857>, rot=<0.2349 -0.2525 0.3549 -0.869>}
 L_cheek_inner {parent=faceAttach, loc=<4.2785 8.761 -1.4435>, rot=<0.0454 -0.6262 0.3191 -0.7099>}
 C_lip_upper_mid {parent=faceAttach, loc=<-0.3154 10.7989 -0.2464>, rot=<0.2097 -0.3713 -0.0897 -0.9001>}
 L_lip_upper_outer {parent=faceAttach, loc=<0.8467 10.1076 -1.4454>, rot=<0.1857 -0.3758 0.3015 -0.8564>}
 L_lip_corner {parent=faceAttach, loc=<1.613 9.1457 -2.3124>, rot=<0.0012 -0.375 -0.0189 -0.9269>}
 C_nose_bridge {parent=faceAttach, loc=<2.3602 13.1515 2.4853>, rot=<-0.2002 -0.3676 0.0825 -0.9044>}
 R_brow_mid {parent=faceAttach, loc=<3.0903 10.342 7.7132>, rot=<-0.1347 -0.3556 -0.2231 -0.8976>}
 R_brow_outer {parent=faceAttach, loc=<1.4121 8.0675 9.1444>, rot=<-0.1251 -0.3913 -0.4176 -0.8104>}
 R_cheek_inner {parent=faceAttach, loc=<-1.4992 8.7305 4.3237>, rot=<-0.1979 -0.0545 -0.2598 -0.9436>}
 R_lip_corner {parent=faceAttach, loc=<-2.32 9.1134 1.692>, rot=<0.0127 -0.39 0.008 -0.9207>}
 C_jaw {parent=faceAttach, loc=<0.3123 3.2483 0.3583>, rot=<0.0044 -0.3827 -0.0044 -0.9239>}
 L_lip_lower_outer {parent=C_jaw, loc=<1.6966 6.8705 -0.8455>, rot=<-0.0146 -0.0279 0.3432 -0.9387>}
 C_lip_lower_mid {parent=C_jaw, loc=<0.0184 7.4191 -0.884>, rot=<0 -0.0099 -0.0021 -0.9999>}
 R_lip_lower_outer {parent=C_jaw, loc=<-1.6982 6.9042 -0.8853>, rot=<0.0021 -0.0171 -0.3152 -0.9489>}
 teeth_lower {parent=C_jaw, loc=<0.0169 5.3598 -2.8149>, rot=<0 0 0 1>}
 tongue {parent=C_jaw, loc=<0.0178 1.5245 -1.0398>, rot=<-0.1568 0 0 -0.9876>}
 R_eye {parent=faceAttach, loc=<2.2537 8.4351 6.6885>, rot=<0.3823 -0.0165 0.9223 -0.0548>}
 R_eye_lid_upper_mid {parent=R_eye, loc=<-0.0465 -1.169 0.2722>, rot=<0.0559 -0.0011 0.0197 -0.9982>, scale=<1 1.0001 0.9999>}
 R_eye_lid_lower_mid {parent=R_eye, loc=<-0.0141 -0.9451 -0.7006>, rot=<0.035 -0.0006 0.0197 -0.9992>}
 L_eye {parent=faceAttach, loc=<6.6314 8.4647 2.3108>, rot=<0.3819 0.0253 0.9227 0.046>}
 L_eye_lid_upper_mid {parent=L_eye, loc=<0.0251 -1.1623 0.2388>, rot=<0.0523 0 0 -0.9986>}
 L_eye_lid_lower_mid {parent=L_eye, loc=<0.0223 -0.923 -0.7488>, rot=<0.0329 0 0 -0.9995>}
 teeth_upper {parent=faceAttach, loc=<0.0954 9.3381 0.1585>, rot=<0.0044 -0.3827 -0.0044 -0.9239>}
 R_lip_upper_outer {parent=faceAttach, loc=<-1.497 10.1359 0.962>, rot=<-0.0848 -0.2926 -0.3215 -0.8966>}
 neck_01_ALT_A {parent=spine_05, loc=<14.3976 4.4897 0>, rot=<0 0 -0.2003 -0.9797>}
 neck_01_ALT_B {parent=neck_01_ALT_A, loc=<2.3802 -1.4414 0>, rot=<0 0 0 1>}
 neck_02_ALT_C {parent=neck_01_ALT_B, loc=<3.03 0.2732 0>, rot=<0 0 0.0504 -0.9987>}
 neck_02_ALT_D {parent=neck_02_ALT_C, loc=<2.6172 -1.0493 0>, rot=<0 0 0 -1>}
 head_ALT_E {parent=neck_02_ALT_D, loc=<3.0025 0.8306 0>, rot=<0 0 0.115 -0.9934>}
 head_ALT_F {parent=head_ALT_E, loc=<3.403 -1.2611 0>, rot=<0 0 0 -1>}
 head_ALT_G {parent=head_ALT_F, loc=<3.2129 -1.2557 0>, rot=<0 0 0 -1>}
 dyn_mustache_mstr {parent=head_ALT_G, loc=<-10.5348 13.1658 0.0007>, rot=<0.4979 0.5021 0.5021 0.4979>}
 dyn_mustache1_l {parent=dyn_mustache_mstr, loc=<0 0 0>, rot=<-0.0699 -0.0899 0.7844 -0.6097>}
 dyn_mustache2_l {parent=dyn_mustache1_l, loc=<7.0687 0 0>, rot=<-0.1506 0.4423 0.0695 -0.8814>}
 dyn_mustache1_r {parent=dyn_mustache_mstr, loc=<0 0 0>, rot=<-0.6094 -0.7846 0.0898 -0.0697>}
 dyn_mustache2_r {parent=dyn_mustache1_r, loc=<7.0235 0 0>, rot=<0.1505 -0.4409 0.0697 -0.8821>}
 dyn_pancake_base_G {parent=head_ALT_G, loc=<0.0643 -0.111 -0.4893>, rot=<0.9564 0.0601 0.0013 0.2859>}
 dyn_pancake_G_side_01 {parent=dyn_pancake_base_G, loc=<0 0 0.7281>, rot=<0 0.7071 0 -0.7071>}
 dyn_syrup_A_01 {parent=dyn_pancake_G_side_01, loc=<8.5922 -3.204 0.9953>, rot=<0.6953 -0.1288 0.7012 0.0915>}
 dyn_syrup_A_02 {parent=dyn_syrup_A_01, loc=<4.9993 0 0>, rot=<0 0 0 -1>}
 dyn_syrup_J_01 {parent=dyn_pancake_G_side_01, loc=<7.5132 -7.3089 1.295>, rot=<0.6794 -0.1961 0.6794 0.1961>}
 dyn_syrup_B_01 {parent=dyn_pancake_G_side_01, loc=<7.733 2.0056 0.7898>, rot=<0.6983 0.1113 0.6983 -0.1113>}
 dyn_pancake_G_side_02 {parent=dyn_pancake_base_G, loc=<0 0 -0.8571>, rot=<0.7071 0 -0.7071 0>}
 dyn_syrup_D_01 {parent=dyn_pancake_G_side_02, loc=<2.5598 -8.8046 2.0837>, rot=<0.5384 -0.4584 0.5384 0.4584>}
 dyn_syrup_E_01 {parent=dyn_pancake_G_side_02, loc=<8.1452 -6.2967 2.6696>, rot=<0.6364 -0.2902 0.6528 0.2908>}
 dyn_syrup_E_02 {parent=dyn_syrup_E_01, loc=<3.0062 0 0>, rot=<0 0 0 1>}
 dyn_syrup_F_01 {parent=dyn_pancake_G_side_02, loc=<11.6594 -0.6548 2.2326>, rot=<0.7021 -0.0838 0.7021 0.0838>}
 dyn_syrup_F_02 {parent=dyn_syrup_F_01, loc=<3.6876 0 0>, rot=<0 0 0 1>}
 dyn_syrup_G_01 {parent=dyn_pancake_G_side_02, loc=<10.9374 5.8375 5.6831>, rot=<0.679 0.1974 0.679 -0.1974>}
 dyn_syrup_G_02 {parent=dyn_syrup_G_01, loc=<3.6876 0 0>, rot=<0 0 0 -1>}
 dyn_syrup_H_01 {parent=dyn_pancake_G_side_02, loc=<0.4857 12.8706 5.9479>, rot=<0.5007 0.4993 0.5007 -0.4993>}
 dyn_syrup_H_02 {parent=dyn_syrup_H_01, loc=<3.6876 0 0>, rot=<0 0 0 -1>}
 dyn_syrup_K_01 {parent=dyn_pancake_G_side_02, loc=<7.2555 10.1429 6.9669>, rot=<0.6272 0.3266 0.6272 -0.3266>}
 dyn_syrup_L_01 {parent=dyn_pancake_G_side_02, loc=<3.5861 12.2714 6.9669>, rot=<0.5771 0.4086 0.5771 -0.4086>}
 dyn_syrup_C_01 {parent=dyn_pancake_base_G, loc=<-0.9409 7.1252 4.401>, rot=<0 0.4606 0.8876 0>}
 dyn_syrup_C_02 {parent=dyn_syrup_C_01, loc=<3.5359 0 0>, rot=<0 0 0 1>}
 dyn_syrup_I_01 {parent=dyn_pancake_base_G, loc=<-0.8837 -10.3325 5.3816>, rot=<0.0056 -0.4831 0.8747 -0.0386>}
 dyn_syrup_I_02 {parent=dyn_syrup_I_01, loc=<4.9647 0 0>, rot=<0 0 0 1>}
 dyn_butter {parent=dyn_pancake_base_G, loc=<2.3355 -2.4049 -2.3632>, rot=<0 0 0 -1>}
 dyn_pancake_base_F {parent=head_ALT_F, loc=<-0.0317 0.0533 -0.4744>, rot=<-0.2476 -0.014 -0.0755 -0.9658>}
 dyn_pancake_F_side_01 {parent=dyn_pancake_base_F, loc=<0 0 0.7281>, rot=<0 0.7071 0 -0.7071>}
 dyn_pancake_F_side_02 {parent=dyn_pancake_base_F, loc=<0 0 -0.8571>, rot=<0.7071 0 -0.7071 0>}
 dyn_pancake_base_E {parent=head_ALT_E, loc=<0.0945 0.0811 0.1443>, rot=<0.4769 0.0251 -0.0505 -0.8772>}
 dyn_pancake_E_side_01 {parent=dyn_pancake_base_E, loc=<0 0 0.7281>, rot=<0 0.7071 0 -0.7071>}
 dyn_pancake_E_side_02 {parent=dyn_pancake_base_E, loc=<0 0 -0.8571>, rot=<0.7071 0 -0.7071 0>}
 dyn_pancake_base_D {parent=neck_02_ALT_D, loc=<0.0312 -0.0499 -1.1561>, rot=<-0.7119 0.0322 0.0198 -0.7013>}
 dyn_pancake_D_side_01 {parent=dyn_pancake_base_D, loc=<0 0 0.7281>, rot=<0 0.7071 0 -0.7071>}
 dyn_pancake_D_side_02 {parent=dyn_pancake_base_D, loc=<0 0 -0.8571>, rot=<-0.415 0 0.9098 0>}
 dyn_pancake_base_C {parent=neck_02_ALT_C, loc=<-0.0229 -0.1209 -1.0537>, rot=<0.7029 -0.0211 0.0138 -0.7108>}
 dyn_pancake_C_side_02 {parent=dyn_pancake_base_C, loc=<0 0 -0.8571>, rot=<0.7071 0 -0.7071 0>}
 faceAttach_upper_ALT_Main {parent=dyn_pancake_base_C, loc=<-0.5395 -1.0703 -2.0305>, rot=<-0.6705 -0.3233 -0.2087 -0.6343>, scale=<0.89 0.89 0.89>}
 R_lip_upper_outer_ALT_Main {parent=faceAttach_upper_ALT_Main, loc=<-3.4503 12.6252 1.948>, rot=<-0.0848 -0.2926 -0.3215 -0.8966>}
 R_lip_corner_ALT_Main {parent=faceAttach_upper_ALT_Main, loc=<-5.3939 10.379 3.791>, rot=<-0.0911 -0.3794 -0.2361 -0.8899>, scale=<1.1236 1.1236 1.1236>}
 L_lip_corner_ALT_Main {parent=faceAttach_upper_ALT_Main, loc=<5.6144 10.41 -6.8698>, rot=<0.1277 -0.3526 0.295 -0.8789>, scale=<1.1236 1.1236 1.1236>}
 L_lip_upper_outer_ALT_Main {parent=faceAttach_upper_ALT_Main, loc=<3.4922 12.6036 -5.1039>, rot=<0.1857 -0.3758 0.3015 -0.8564>}
 C_lip_upper_mid_ALT_Main {parent=faceAttach_upper_ALT_Main, loc=<-0.0307 13.6662 -1.6785>, rot=<0.2097 -0.3713 -0.0897 -0.9001>}
 dyn_pancake_base_B {parent=neck_01_ALT_B, loc=<0.1426 -0.155 -0.8801>, rot=<0.6424 -0.0441 0.0251 -0.7647>}
 dyn_pancake_B_side_02 {parent=dyn_pancake_base_B, loc=<0 0 -0.8571>, rot=<0.7071 0 -0.7071 0>}
 faceAttach_lower_ALT_Main {parent=dyn_pancake_base_B, loc=<2.1379 -1.1862 -1.2455>, rot=<-0.6278 -0.3455 -0.1552 -0.68>, scale=<0.89 0.89 0.89>}
 C_jaw_ALT_Main {parent=faceAttach_lower_ALT_Main, loc=<1.1088 1.9431 -0.5359>, rot=<0.0044 -0.3827 -0.0044 -0.9239>}
 L_lip_lower_outer_ALT_Main {parent=C_jaw_ALT_Main, loc=<4.9976 10.6684 -1.4685>, rot=<-0.0146 -0.0279 0.3432 -0.9387>}
 C_lip_lower_mid_ALT_Main {parent=C_jaw_ALT_Main, loc=<0.0542 11.522 -1.6494>, rot=<0 -0.0099 -0.0021 -0.9999>}
 R_lip_lower_outer_ALT_Main {parent=C_jaw_ALT_Main, loc=<-5.0021 10.7209 -1.5369>, rot=<0.0021 -0.0171 -0.3152 -0.9489>}
 dyn_pancake_base_A {parent=neck_01_ALT_A, loc=<0.1702 0.3324 -0.6781>, rot=<0 0 0.0383 -0.9993>}
 dyn_pancake_A_side_01 {parent=dyn_pancake_base_A, loc=<0 0 0.7281>, rot=<0 0.7071 0 -0.7071>}
 dyn_pancake_A_side_02 {parent=dyn_pancake_base_A, loc=<0 0 -0.8571>, rot=<0.7071 0 -0.7071 0>}
 JawSpace {parent=root, type=Null, loc=<-0.0151 1.0288 151.6187>, rot=<-0.0074 0 0 1>, scale=<0.89 0.89 0.89>}
 Jaw {parent=JawSpace, type=Control, loc=<0 0 0>, rot=<0 0 0 1>}
]
