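# Physics asset for the mustache part of the Flapjack Wrangler male medium character, defining the bodies that let it move as simulated bones.
# Capsule bodies sit on a master mustache bone and on two-segment left and right chains.
# Angular constraints with tight swing and twist limits link each inner segment to the master and each tip to its inner segment, allowing only slight wobble on both sides.
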PhysicsAsset /Game/Characters/Player/Male/Medium/Bodies/M_MED_Flapjack_Wrangler/Meshes/Parts/M_MED_Flapjack_Wrangler_Physics_Mustache
bodies {
 dyn_mustache1_l [{shapeType=Capsule, length=3, radius=1.5, center=<2 0 0>, rotation=<rot 90 0 0>}]
 dyn_mustache1_r [{shapeType=Capsule, length=3, radius=1.5, center=<2 0 0>, rotation=<rot 90 0 0>}]
 dyn_mustache2_l [{shapeType=Capsule, length=3, radius=1.5, center=<2 0 0>, rotation=<rot 90 0 0>}]
 dyn_mustache2_r [{shapeType=Capsule, length=3, radius=1.5, center=<2 0 0>, rotation=<rot 90 0 0>}]
 dyn_mustache_mstr [{shapeType=Capsule, length=1, radius=1}]
}
constraints [
 {bone1Name=dyn_mustache1_l, bone2Name=dyn_mustache_mstr, swing1LimitDegrees=5, swing2LimitDegrees=5, twistLimitDegrees=2}
 {bone1Name=dyn_mustache2_l, bone2Name=dyn_mustache1_l, swing1LimitDegrees=5, swing2LimitDegrees=5, twistLimitDegrees=2}
 {bone1Name=dyn_mustache1_r, bone2Name=dyn_mustache_mstr, swing1LimitDegrees=5, swing2LimitDegrees=5, twistLimitDegrees=2}
 {bone1Name=dyn_mustache2_r, bone2Name=dyn_mustache1_r, swing1LimitDegrees=5, swing2LimitDegrees=5, twistLimitDegrees=2}
]
