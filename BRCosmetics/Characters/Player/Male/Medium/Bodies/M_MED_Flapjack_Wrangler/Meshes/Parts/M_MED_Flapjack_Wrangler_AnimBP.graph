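# Animation Blueprint for the Flapjack Wrangler cosmetic body, adding secondary motion to its syrup and pancakes.
# It copies the parent mesh pose, partially drives ALT neck and head bones from ref-pose deltas, then runs head-relative RigidBody simulations.
# Frame time softens syrup gravity and inertia on slow frames, and the Gunspinner syrup emote layers a time-synced drip animation over syrup bones.
# Control Rig and variant-toggled dark-syrup physics nodes sit outside the output chain.
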AnimBlueprint /BRCosmetics/Characters/Player/Male/Medium/Bodies/M_MED_Flapjack_Wrangler/Meshes/Parts/M_MED_Flapjack_Wrangler_AnimBP
target_skeleton=&/M_MED_Flapjack_Wrangler_Skeleton
graphs [
 AnimGraph {nodes [
   Control Rig() {
    alphaBoolBlend {blendOption=Linear}
    controlRigAssetReference {blueprintRigClass=/Game/Characters/Player/Male/Medium/Bodies/M_MED_Flapjack_Wrangler/Meshes/Parts/M_MED_Flapjack_Wrangler_CtrlRig.M_MED_Flapjack_Wrangler_CtrlRig_C}
   }
   n1 = Get DarkUpdatesActive() -> (DarkUpdatesActive:bool)
   n2 = RigidBody!(ComponentPose, bAlphaBoolEnabled:bool=@n1, ExternalForce) -> (Pose) {
    alphaInputType=Bool
    bClampLinearTranslationLimitToRefPose=true
    bOverrideWorldGravity=true
    overridePhysicsAsset=/BRCosmetics/Characters/Player/Male/Medium/Bodies/M_MED_Flapjack_Wrangler_Dark_Updates/Meshes/Parts/M_MED_Flapjack_Wrangler_Physics_SyrupDark
    simulationSpace=BaseBoneSpace
    alphaBoolBlend {blendOption=Linear}
    baseBoneRef {boneName=head}
    componentAppliedLinearAccClamp=<3000 3000 3000>
    componentLinearAccScale=<0.35 0.3 0.3>
    componentLinearVelScale=<0.35 0.3 0.3>
    overrideWorldGravity {z=-980}
   }
   n3 = RigidBody(ComponentPose=@n2, bAlphaBoolEnabled=@n1) {
    alphaInputType=Bool
    bClampLinearTranslationLimitToRefPose=true
    bOverrideWorldGravity=true
    overridePhysicsAsset=/BRCosmetics/Characters/Player/Male/Medium/Bodies/M_MED_Flapjack_Wrangler_Dark_Updates/Meshes/Parts/M_MED_Flapjack_Wrangler_Dark_Updates_Physics
    simulationSpace=BaseBoneSpace
    alphaBoolBlend {blendOption=Linear}
    baseBoneRef {boneName=head}
    componentAppliedLinearAccClamp=<5000 5000 5000>
    componentLinearAccScale=<0.4 0.4 0.4>
    componentLinearVelScale=<0.3 0.3 0.3>
    overrideWorldGravity {z=-980}
   }
   RigidBody(ComponentPose=@n3, bAlphaBoolEnabled=@n1) {
    alphaInputType=Bool
    bClampLinearTranslationLimitToRefPose=true
    bOverrideWorldGravity=true
    overridePhysicsAsset=/Game/Characters/Player/Male/Medium/Bodies/M_MED_Flapjack_Wrangler/Meshes/Parts/M_MED_Flapjack_Wrangler_Physics_Mustache
    simulationSpace=BaseBoneSpace
    alphaBoolBlend {blendOption=Linear}
    baseBoneRef {boneName=head}
    componentAppliedLinearAccClamp=<3000 3000 3000>
    componentLinearAccScale=<0.35 0.3 0.3>
    componentLinearVelScale=<0.35 0.3 0.3>
    overrideWorldGravity {z=-980}
   }
   n4 = Get Neck 01 Rot() -> (Neck 01 Rot:Rotator)
   n5 = "Transform (Modify) Bone - Bone: neck_01_ALT_A"(
    ComponentPose=Local To Component(LocalPose=Copy Pose From Mesh() {bCopyCurves=true, bUseAttachedParent=true})
    Alpha=0.5
    Translation="0,0,0"
    Rotation=@n4
    Scale="1,1,1"
   ) {rotationMode=BMM_Replace, rotationSpace=BCS_BoneSpace, alphaBoolBlend {blendOption=Linear}, boneToModify {boneName=neck_01_ALT_A}}
   n6 = "Transform (Modify) Bone - Bone: neck_01_ALT_B"(ComponentPose=@n5, Alpha=0.5, Translation="0,0,0", Rotation=@n4, Scale="1,1,1") {rotationMode=BMM_Replace, rotationSpace=BCS_BoneSpace, alphaBoolBlend {blendOption=Linear}, boneToModify {boneName=neck_01_ALT_B}}
   n7 = Get Neck 02 Rot() -> (Neck 02 Rot:Rotator)
   n8 = "Transform (Modify) Bone - Bone: neck_02_ALT_C"(ComponentPose=@n6, Alpha=0.5, Translation="0,0,0", Rotation=@n7, Scale="1,1,1") {rotationMode=BMM_Replace, rotationSpace=BCS_BoneSpace, alphaBoolBlend {blendOption=Linear}, boneToModify {boneName=neck_02_ALT_C}}
   n9 = "Transform (Modify) Bone - Bone: neck_02_ALT_D"(ComponentPose=@n8, Alpha=0.5, Translation="0,0,0", Rotation=@n7, Scale="1,1,1") {rotationMode=BMM_Replace, rotationSpace=BCS_BoneSpace, alphaBoolBlend {blendOption=Linear}, boneToModify {boneName=neck_02_ALT_D}}
   n10 = Get head rot() -> (head rot:Rotator)
   n11 = "Transform (Modify) Bone - Bone: head_ALT_E"(ComponentPose=@n9, Alpha=0.2, Translation="0,0,0", Rotation=@n10, Scale="1,1,1") {rotationMode=BMM_Replace, rotationSpace=BCS_BoneSpace, alphaBoolBlend {blendOption=Linear}, boneToModify {boneName=head_ALT_E}}
   n12 = "Transform (Modify) Bone - Bone: head_ALT_F"(ComponentPose=@n11, Alpha=0.2, Translation="0,0,0", Rotation=@n10, Scale="1,1,1") {rotationMode=BMM_Replace, rotationSpace=BCS_BoneSpace, alphaBoolBlend {blendOption=Linear}, boneToModify {boneName=head_ALT_F}}
   n13 = "Transform (Modify) Bone - Bone: head_ALT_G"(ComponentPose=@n12, Alpha=0.2, Translation="0,0,0", Rotation=@n10, Scale="1,1,1") {rotationMode=BMM_Replace, rotationSpace=BCS_BoneSpace, alphaBoolBlend {blendOption=Linear}, boneToModify {boneName=head_ALT_G}}
   n14 = Get Gravity_override_from_fps() -> (Gravity_override_from_fps:Vector)
   n15 = Get Component_scale_from_fps() -> (Component_scale_from_fps:Vector)
   n16 = RigidBody!(
    ComponentPose=@n13
    Alpha:float
    OverrideWorldGravity:Vector=@n14
    ExternalForce
    ComponentLinearAccScale:Vector=@n15
    ComponentLinearVelScale:Vector=@n15
   ) -> (Pose) {
    bClampLinearTranslationLimitToRefPose=true
    bOverrideWorldGravity=true
    overridePhysicsAsset=&/M_MED_Flapjack_Wrangler_Physics_Syrup
    simulationSpace=BaseBoneSpace
    alphaBoolBlend {blendOption=Linear, blendOutTime=10}
    alphaScaleBiasClamp {bInterpResult=true}
    baseBoneRef {boneName=head}
    componentAppliedLinearAccClamp=<3000 3000 3000>
    componentLinearAccScale=<0.15 0.15 0.15>
    componentLinearVelScale=<0.15 0.15 0.15>
    overrideWorldGravity {z=-350}
   }
   n17 = RigidBody(ComponentPose=@n16, Alpha=0, OverrideWorldGravity=@n14, ComponentLinearAccScale=@n15, ComponentLinearVelScale=@n15) {
    bClampLinearTranslationLimitToRefPose=true
    bOverrideWorldGravity=true
    overridePhysicsAsset=&/M_MED_Flapjack_Wrangler_Physics_Cakes
    simulationSpace=BaseBoneSpace
    alphaBoolBlend {blendOption=Linear, blendOutTime=10}
    alphaScaleBiasClamp {bInterpResult=true}
    baseBoneRef {boneName=head}
    componentAppliedLinearAccClamp=<3000 3000 3000>
    componentLinearAccScale=<0.15 0.15 0.15>
    componentLinearVelScale=<0.15 0.15 0.15>
    overrideWorldGravity {z=-350}
   }
   n18 = RigidBody(ComponentPose=@n17, OverrideWorldGravity="0,0,-350", ComponentLinearAccScale="0.15,0.15,0.15", ComponentLinearVelScale="0.15,0.15,0.15") {
    bClampLinearTranslationLimitToRefPose=true
    bOverrideWorldGravity=true
    overridePhysicsAsset=&M_MED_Flapjack_Wrangler_Physics_Cakes
    simulationSpace=BaseBoneSpace
    alphaBoolBlend {blendOption=Linear, blendOutTime=10}
    alphaScaleBiasClamp {bInterpResult=true}
    baseBoneRef {boneName=head}
    componentAppliedLinearAccClamp=<3000 3000 3000>
    componentLinearAccScale=<0.15 0.15 0.15>
    componentLinearVelScale=<0.15 0.15 0.15>
    overrideWorldGravity {z=-350}
   }
   n19 = Layered blend per bone(
    BasePose=Component To Local(ComponentPose=@n18)
    BlendPoses_0="Sequence Evaluator (by time) 'M_MED_Flapjack_Wrangler_EmoteSyrupDrip'"!(ExplicitTime:float=Get customEmotePosition() -> (customEmotePosition:double)) -> (Pose:Pose Link Structure) {
     is=title AnimGraphNode_SequenceEvaluator
     sequence=&/M_MED_Flapjack_Wrangler_EmoteSyrupDrip
    }
    BlendWeights_0=Get isCustomEmotePlaying() -> (isCustomEmotePlaying:double)
   ) {layerSetup [{branchFilters [
       {boneName=dyn_syrup_I_02}
       {boneName=dyn_syrup_J_01}
       {boneName=dyn_syrup_A_02}
       {boneName=dyn_syrup_B_01}
       {boneName=dyn_syrup_C_02}
       {boneName=dyn_syrup_D_01}
       {boneName=dyn_syrup_E_02}
       {boneName=dyn_syrup_F_02}
      ]}]}
   Output Pose(Result=@n19) {layerGroup=None}
   NOT Boolean(A=@n1)
  ]}
 EventGraph {nodes [
   n1 = Event BlueprintUpdateAnimation()
   n2 = Get Skel_Mesh()
   > Set head rot(head rot:Rotator=BreakTransform(InTransform=GetDeltaTransformFromRefPose(self=@n2, BoneName=head, BaseName=neck_02)).Rotation) -> (Output_Get:Rotator)
   > Set Neck 01 Rot(Neck 01 Rot:Rotator=BreakTransform(InTransform=GetDeltaTransformFromRefPose(self=@n2, BoneName=neck_01, BaseName=spine_05)).Rotation) -> (Output_Get:Rotator)
   > n3 = Set Neck 02 Rot(Neck 02 Rot:Rotator=BreakTransform(InTransform=GetDeltaTransformFromRefPose(self=@n2, BoneName=neck_02, BaseName=neck_01)).Rotation) -> (Output_Get:Rotator)
   n4 = Set Skel_Mesh(execute=@n9.then, Skel_Mesh=GetOwnerSkeletalMeshComponent())
   n5 = GetAnimInstance(self=@n4)
   Set isCustomEmotePlaying(
    execute=@n3.then
    isCustomEmotePlaying:double=SelectFloat(
     A=1
     B=0
     bPickA=Montage_IsActive!(
      self:Anim Instance Object Reference=@n5
      Montage:Anim Montage Object Reference=/BRCosmetics/Animation/Game/MainPlayer/Emotes/Gunspinner_Syrup/Emote_Gunspinner_Syrup_CMM_M
     ) -> (ReturnValue:bool)
    )
   ) -> (Output_Get:double)
   > Set customEmotePosition(customEmotePosition:double=Montage_GetPosition!(
     self:Anim Instance Object Reference=@n5
     Montage:Anim Montage Object Reference=&Emote_Gunspinner_Syrup_CMM_M
    ) -> (ReturnValue:float)) -> (Output_Get:double)
   > Set Gravity_override_from_fps(Gravity_override_from_fps:Vector=MakeVector(Y=0, Z=MapRangeClamped(Value=@n1.DeltaTimeX, InRangeA=0.022, InRangeB=0.028, OutRangeA=-980, OutRangeB=-350))) -> (Output_Get:Vector)
   n6 = MapRangeClamped(Value=@n1.DeltaTimeX, InRangeA=0.022, InRangeB=0.028, OutRangeA=0.35, OutRangeB=0.15)
   > Set Component_scale_from_fps(Component_scale_from_fps:Vector=MakeVector(X=@n6, Y=@n6, Z=@n6)) -> (Output_Get:Vector)
   > Set NotFrontEnd(NotFrontEnd:bool="Not Equal (Boolean)"(A=Get bIsInFrontEnd()) -> (ReturnValue=~)) -> (Output_Get:bool)
   Event BlueprintPostEvaluateAnimation() {is=title K2Node_Event}
   n7 = GetPartSkeletalMeshComponent()
   > Is Valid(InputObject=@n7)
   > Delay(Duration=0)
   > SetVectorParameterValueOnMaterials!(
    execute
    self:Mesh Component Object Reference=@n7
    ParameterName:Name=headSocketLocation
    ParameterValue:Vector=GetSocketLocation!(self:Scene Component Object Reference=@n7, InSocketName:Name=headSocket) -> (ReturnValue:Vector)
   ) -> (then)
   Event OnCharacterCustomizationLoaded()
   > n8 = Branch(Condition=CharacterItemDefinitionHasAnyMetaTag(Tags=Make Array("[0]"=(TagName="Cosmetics.Variant.Property.Stage4"))))
   > Set DarkUpdatesActive(DarkUpdatesActive:bool=true) -> (Output_Get:bool)
   Set DarkUpdatesActive(execute=@n8.else, DarkUpdatesActive=false)
   n9 = Event BlueprintInitializeAnimation()
  ]}
]
variables [
 Component_scale_from_fps {type=Vector}
 customEmotePosition {type=number}
 DarkUpdatesActive {type=boolean}
 Gravity_override_from_fps {type=Vector}
 head rot {type=Rotator}
 isCustomEmoteNotPlaying {type=number}
 isCustomEmotePlaying {type=number}
 Neck 01 Rot {type=Rotator}
 Neck 02 Rot {type=Rotator}
 NotFrontEnd {type=boolean}
 Skel_Mesh {type=SkeletalMeshComponent}
]
